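# Control Rig for the MaltBurger male medium character body, correcting deformation of the legs and skirt or coat bones.
# The main graph runs FortDeformRig and Leg ROM, then drives thigh twist correctives from thigh and calf rotation.
# Skirt bones follow thighs and pelvis through weighted parent constraints.
# Library functions remap joint angles and spine height into weights for additive coat offsets, alongside interpolated movement-state flags like crouching and sliding.
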ControlRigBlueprint /BRCosmetics/Characters/Player/Male/Medium/Bodies/M_MED_MaltBurger/Meshes/M_MED_MaltBurger_CtrlRig
graphs [
 Multiply_10_ContainedGraph {nodes [
   n1 = RigVMFunctionEntryNode()
   RigVMFunctionReturnNode(Result=&RigVMFunction_MathDoubleMul(B=@n1.C, A=&RigVMFunction_MathDoubleMul(A=@n1.A, B=@n1.B).Result).Result)
  ]}
 Multiply_10_ContainedGraph {nodes [
   n1 = RigVMFunctionEntryNode()
   RigVMFunctionReturnNode(Result=&RigVMFunction_MathDoubleMul(B=@n1.C, A=&RigVMFunction_MathDoubleMul(A=@n1.A, B=@n1.B).Result).Result)
  ]}
 Multiply_11_ContainedGraph {nodes [
   n1 = RigVMFunctionEntryNode()
   RigVMFunctionReturnNode(Result=&RigVMFunction_MathDoubleMul(B=@n1.C, A=&RigVMFunction_MathDoubleMul(A=@n1.A, B=@n1.B).Result).Result)
  ]}
 Multiply_9_ContainedGraph {nodes [
   n1 = RigVMFunctionEntryNode()
   Multiply = &RigVMFunction_MathDoubleMul(A=@n1.A, B=@n1.B)
   RigVMFunctionReturnNode(Result=&RigVMFunction_MathDoubleMul(B=@n1.C, A=@Multiply.Result).Result)
  ]}
 New Function_ContainedGraph {nodes [
   n1 = RigVMFunctionEntryNode()
   RigVMFunction_MathQuaternionToEuler = &RigVMFunction_MathQuaternionToEuler(
    RotationOrder=XYZ
    Value=&RigUnit_GetRelativeTransformForItem(
     Child=(Type=Bone,Name="hand_l")
     Parent=(Type=Null,Name="lowerarm_l")
     bChildInitial=False
     bParentInitial=False
     Child.Name=@n1.Child
     Parent.Name=@n1.Parent
    ).RelativeTransform.Rotation
   )
   Remap = &RigVMFunction_MathDoubleRemap(
    TargetMaximum=1
    TargetMinimum=0
    Value=@RigVMFunction_MathQuaternionToEuler.Result.Y
    SourceMaximum=@n1.Source Maximum
    SourceMinimum=@n1.Source Minimum
   )
   > RigVMFunctionReturnNode(Delta Rotation=@Remap.Result)
  ]}
 New Function_ContainedGraph {nodes [
   n1 = RigVMFunctionEntryNode()
   RigVMFunction_MathQuaternionToEuler = &RigVMFunction_MathQuaternionToEuler(
    RotationOrder=XYZ
    Value=&RigUnit_GetRelativeTransformForItem(
     Child=(Type=Bone,Name="hand_l")
     Parent=(Type=Null,Name="lowerarm_l")
     bChildInitial=False
     bParentInitial=False
     Child.Name=@n1.Child
     Parent.Name=@n1.Parent
    ).RelativeTransform.Rotation
   )
   Remap = &RigVMFunction_MathDoubleRemap(
    TargetMaximum=1
    TargetMinimum=0
    Value=@RigVMFunction_MathQuaternionToEuler.Result.X
    SourceMaximum=@n1.Source Maximum
    SourceMinimum=@n1.Source Minimum
   )
   > RigVMFunctionReturnNode(Delta Rotation=@Remap.Result)
  ]}
 New Function_ContainedGraph {nodes [
   n1 = RigVMFunctionEntryNode()
   RigVMFunction_MathQuaternionToEuler = &RigVMFunction_MathQuaternionToEuler(
    RotationOrder=XYZ
    Value=&RigUnit_GetRelativeTransformForItem(
     Child=(Type=Bone,Name="hand_l")
     Parent=(Type=Null,Name="lowerarm_l")
     bChildInitial=False
     bParentInitial=False
     Child.Name=@n1.Child
     Parent.Name=@n1.Parent
    ).RelativeTransform.Rotation
   )
   Remap = &RigVMFunction_MathDoubleRemap(
    TargetMaximum=1
    TargetMinimum=0
    Value=@RigVMFunction_MathQuaternionToEuler.Result.Z
    SourceMaximum=@n1.Source Maximum
    SourceMinimum=@n1.Source Minimum
   )
   > RigVMFunctionReturnNode(Delta Rotation=@Remap.Result)
  ]}
 New Function_ContainedGraph {nodes [
   n1 = RigVMFunctionEntryNode()
   RigVMFunction_ControlFlowBranch = &RigVMFunction_ControlFlowBranch(Condition=@n1.Debug Controls, ExecuteContext=@n1.ExecuteContext)
   RigVMFunctionReturnNode(ExecuteContext=@RigVMFunction_ControlFlowBranch.Completed)
   RigUnit_GetTransform = &RigUnit_GetTransform(Item=(Type=Control,Name="thigh_r_ctrl"), bInitial=False)
   &RigUnit_SetTransform(
    Item=(Type=Bone,Name="thigh_r")
    Weight=1
    bInitial=False
    Value=@RigUnit_GetTransform.Transform
    ExecutePin=@RigVMFunction_ControlFlowBranch.True
   )
   > &RigUnit_SetTransform(
    Item=(Type=Bone,Name="thigh_l")
    Weight=1
    bInitial=False
    Value=&RigUnit_GetTransform(Item=(Type=Control,Name="thigh_l_ctrl"), bInitial=False).Transform
   )
   > &RigUnit_SetTransform(
    Item=(Type=Bone,Name="calf_l")
    Weight=1
    bInitial=False
    Value=&RigUnit_GetTransform(Item=(Type=Control,Name="calf_l_ctrl"), bInitial=False).Transform
   )
   > &RigUnit_SetTransform(
    Item=(Type=Bone,Name="calf_r")
    Weight=1
    bInitial=False
    Value=&RigUnit_GetTransform(Item=(Type=Control,Name="calf_r_ctrl"), bInitial=False).Transform
   )
   > &RigUnit_SetTransform(
    Item=(Type=Bone,Name="spine_05")
    Weight=1
    bInitial=False
    Value=&RigUnit_GetTransform(Item=(Type=Control,Name="spine_05_ctrl"), bInitial=False).Transform
   )
   &RigUnit_SetTransform(
    ExecutePin=()
    Item=(Type=Bone,Name="upperarm_l")
    Weight=1
    bInitial=False
    Value=&RigUnit_GetTransform(Item=(Type=Control,Name="upperarm_l_ctrl"), bInitial=False).Transform
   )
   > &RigUnit_SetTransform(
    Item=(Type=Bone,Name="upperarm_r")
    Weight=1
    bInitial=False
    Value=&RigUnit_GetTransform(Item=(Type=Control,Name="upperarm_r_ctrl"), bInitial=False).Transform
   )
  ]}
 New Function_ContainedGraph {nodes [
   RigVMFunctionEntryNode()
   > RigVMFunction_Sequence = RigVMAggregateNode()
   Delta Rotation Y = RigVMFunctionReferenceNode(Child=thigh_l, Parent=thigh_l_null, Source Maximum=-75, ExecuteContext=@RigVMFunction_Sequence.A)
   > RigVMVariableNode(Value=@"Delta Rotation Y".Delta Rotation) {variable=Thigh_L_SideRotation}
   > Delta Rotation Y_1 = RigVMFunctionReferenceNode(Child=thigh_r, Parent=thigh_r_null, Source Maximum=-75)
   > RigVMVariableNode(Value=@"Delta Rotation Y_1".Delta Rotation) {variable=Thigh_R_SideRotation}
   > RigVMVariableNode(Value=&RigVMFunction_MathDoubleRemap(
     SourceMaximum=55
     SourceMinimum=65
     TargetMaximum=1.4
     TargetMinimum=0
     Value=&RigUnit_GetRelativeTransformForItem(Child=(Type=Bone,Name="spine_04"), Parent=(Type=Bone,Name="root"), bChildInitial=False, bParentInitial=False).RelativeTransform.Translation.Z
    ).Result) {variable=SpineToGround}
   > RigVMVariableNode(Value=&RigVMFunction_MathDoubleRemap(
     SourceMaximum=70
     SourceMinimum=77
     TargetMaximum=1.4
     TargetMinimum=0
     Value=&RigUnit_GetRelativeTransformForItem(Child=(Type=Bone,Name="spine_04"), Parent=(Type=Bone,Name="root"), bChildInitial=False, bParentInitial=False).RelativeTransform.Translation.Z
    ).Result) {variable=PelvisToGround}
   > Delta Rotation Z = RigVMFunctionReferenceNode(Child=thigh_l, Parent=thigh_l_null, Source Maximum=70, Source Minimum=5)
   > RigVMVariableNode(Value=@"Delta Rotation Z".Delta Rotation) {variable=Thigh_L_BackRotation}
   > Delta Rotation Z_1 = RigVMFunctionReferenceNode(Child=thigh_r, Parent=thigh_r_null, Source Maximum=70, Source Minimum=5)
   > RigVMVariableNode(Value=@"Delta Rotation Z_1".Delta Rotation) {variable=Thigh_R_BackRotation}
   > Delta Rotation Z_2 = RigVMFunctionReferenceNode(Child=calf_l, Parent=calf_l_null, Source Maximum=90, Source Minimum=30)
   > RigVMVariableNode(Value=@"Delta Rotation Z_2".Delta Rotation) {variable=Calf_L_BackRotation}
   > Delta Rotation Z_3 = RigVMFunctionReferenceNode(Child=calf_r, Parent=calf_r_null, Source Maximum=90, Source Minimum=30)
   > RigVMVariableNode(Value=@"Delta Rotation Z_3".Delta Rotation) {variable=Calf_R_BackRotation}
   > Delta Rotation X = RigVMFunctionReferenceNode(Child=spine_05, Parent=spine_05_null, Source Maximum=-30, Source Minimum=10)
   > RigVMVariableNode(Value=@"Delta Rotation X".Delta Rotation) {variable=spineTwist_R}
   RigVMFunctionReturnNode(ExecuteContext=@RigVMFunction_Sequence.C)
  ]}
 New Function_ContainedGraph {nodes [
   RigVMFunctionEntryNode()
   > RigVMFunction_Sequence = RigVMAggregateNode()
   n1 = &RigVMFunction_AlphaInterp(
    InterpSpeedDecreasing=10
    InterpSpeedIncreasing=10
    Scale=1
    bClampResult=true
    Value=&RigVMFunction_MathBoolToFloat(Value=RigVMVariableNode() {variable=isSkydiving_CR}.Value).Result
   )
   &RigUnit_ModifyTransforms(
    ItemToModify="((Item=(Type=Bone,Name=\"base_coatFront_r\"),Transform=(Rotation=(X=0,Y=0,Z=0.0188,W=-0.9998),Translation=(X=0,Y=3.9999,Z=0),Scale3D=(X=1,Y=1,Z=1))),(Item=(Type=Bone,Name=\"dyn_coatFront_02_r\"),Transform=(Rotation=(X=0,Y=0,Z=0.0119,W=0.9999),Translation=(X=0, ...(truncated)"
    Weight=@n1.Result
    ExecutePin=@RigVMFunction_Sequence.F
   )
   > &RigUnit_ModifyTransforms(
    ItemToModify="((Item=(Type=Bone,Name=\"dyn_coatBack_02_r\"),Transform=(Rotation=(X=0.0622,Y=-0.012,Z=0.1886,W=-0.98),Translation=(X=0,Y=0,Z=0),Scale3D=(X=1,Y=1,Z=1))),(Item=(Type=Bone,Name=\"dyn_coatBack_02_l\"),Transform=(Rotation=(X=-0.0881,Y=0.0402,Z=0.1526,W=-0.9835),Tr ...(truncated)"
    Weight=@n1.Result
   )
   n2 = &RigVMFunction_AlphaInterp(
    InterpSpeedDecreasing=10
    InterpSpeedIncreasing=10
    Scale=1
    bClampResult=true
    Value=&RigVMFunction_MathBoolToFloat(Value=RigVMVariableNode() {variable=isEmoting_CR}.Value).Result
   )
   > &RigUnit_ModifyTransforms(
    ItemToModify=((Item=(Type=Bone,Name="base_coatFront_l"),Transform=(Rotation=(X=0,Y=0,Z=-0.0415,W=0.9991),Translation=(X=0,Y=0,Z=0),Scale3D=(X=1,Y=1,Z=1))))
    Weight=@n2.Result
   )
   > &RigUnit_ModifyTransforms(
    ItemToModify="((Item=(Type=Bone,Name=\"base_coatBack_r\"),Transform=(Rotation=(X=0,Y=0,Z=0.7626,W=-0.6468),Translation=(X=0,Y=0,Z=0),Scale3D=(X=1,Y=1,Z=1))),(Item=(Type=Bone,Name=\"base_coatBack_l\"),Transform=(Rotation=(X=0,Y=0,Z=-0.6423,W=0.7664),Translation=(X=0,Y=0,Z=0) ...(truncated)"
    Weight=&RigVMFunction_MathDoubleMul(A=@n2.Result, B=RigVMVariableNode() {variable=SpineToGround}.Value).Result
   )
   n3 = &RigVMFunction_AlphaInterp(
    InterpSpeedDecreasing=10
    InterpSpeedIncreasing=10
    Scale=1
    bClampResult=true
    Value=&RigVMFunction_MathBoolToFloat(Value=RigVMVariableNode() {variable=isOnRail_CR}.Value).Result
   )
   > &RigUnit_ModifyTransforms(
    ItemToModify="((Item=(Type=Bone,Name=\"base_coatBack_l\"),Transform=(Rotation=(X=0,Y=0,Z=0.3729,W=-0.9279),Translation=(X=0,Y=0,Z=0),Scale3D=(X=1,Y=1,Z=1))),(Item=(Type=Bone,Name=\"base_coatBack_r\"),Transform=(Rotation=(X=0,Y=0,Z=0.3151,W=-0.9491),Translation=(X=0,Y=0,Z=0) ...(truncated)"
    Weight=&RigVMFunction_MathDoubleMul(A=@n3.Result, B=RigVMVariableNode() {variable=PelvisToGround}.Value).Result
   )
   RigVMFunctionReturnNode(ExecuteContext=@RigVMFunction_Sequence.H)
   &RigUnit_ParentConstraint(
    AdvancedSettings=(RotationOrderForFilter=XZY)
    Child=(Type=Bone,Name="MSTR_skirt_side_l")
    Parents=((Item=(Type=Bone,Name="thigh_twist_01_l"),Weight=0.2),(Item=(Type=Bone,Name="pelvis"),Weight=0.8))
    Weight=1
    ExecutePin=@RigVMFunction_Sequence.A
   )
   > &RigUnit_ParentConstraint(
    AdvancedSettings=(RotationOrderForFilter=XZY)
    Child=(Type=Bone,Name="MSTR_skirt_fr_rear_l")
    Filter=(TranslationFilter=(bX=false,bY=false,bZ=false),RotationFilter=(bX=True,bY=True,bZ=True),ScaleFilter=(bX=True,bY=True,bZ=True))
    Parents=((Item=(Type=Bone,Name="thigh_twist_01_l"),Weight=0.25),(Item=(Type=Bone,Name="pelvis"),Weight=0.75))
    Weight=1
   )
   n4 = RigVMDispatchNode(BoneName=root, Default=1, Name=deformRig_l_thigh_up_alpha) {template="DISPATCH_RigDispatch_GetAnimAttribute(in Name,in BoneName,in Default,out Value,out Found)"}
   > &RigUnit_ModifyTransforms(
    ItemToModify="((Item=(Type=Bone,Name=\"MSTR_skirt_fr_rear_l\"),Transform=(Rotation=(X=-0.302,Y=-0.0945,Z=0.3172,W=-0.894),Translation=(X=0,Y=0,Z=0),Scale3D=(X=1,Y=1,Z=1))),(Item=(Type=Bone,Name=\"MSTR_skirt_side_l\"),Transform=(Rotation=(X=0.1359,Y=0.0223,Z=-0.1992,W=0.9702 ...(truncated)"
    Weight=@n4.Value
   )
   > &RigUnit_ModifyTransforms(
    ItemToModify="((Item=(Type=Bone,Name=\"MSTR_skirt_fr_r\"),Transform=(Rotation=(X=0.1712,Y=0.0072,Z=0.0958,W=-0.9805),Translation=(X=-0.1648,Y=0,Z=2.9696),Scale3D=(X=1,Y=1,Z=1))),(Item=(Type=Bone,Name=\"MSTR_skirt_fr_rear_r\"),Transform=(Rotation=(X=0,Y=0,Z=0,W=1),Translatio ...(truncated)"
    Weight=@n4.Value
   )
   n5 = RigVMVariableNode() {variable=Thigh_L_SideRotation}
   > &RigUnit_ParentConstraint(
    AdvancedSettings=(RotationOrderForFilter=XZY)
    Child=(Type=Bone,Name="MSTR_skirt_side_l")
    Filter=(TranslationFilter=(bX=false,bY=false,bZ=false),RotationFilter=(bX=True,bY=True,bZ=True),ScaleFilter=(bX=True,bY=True,bZ=True))
    Parents=((Item=(Type=Bone,Name="thigh_twist_01_l"),Weight=0.8),(Item=(Type=Bone,Name="pelvis"),Weight=0.2))
    Weight=@n5.Value
   )
   > &RigUnit_ParentConstraint(
    AdvancedSettings=(RotationOrderForFilter=XZY)
    Child=(Type=Bone,Name="MSTR_skirt_fr_rear_l")
    Filter=(TranslationFilter=(bX=false,bY=false,bZ=false),RotationFilter=(bX=True,bY=True,bZ=True),ScaleFilter=(bX=True,bY=True,bZ=True))
    Parents=((Item=(Type=Bone,Name="thigh_twist_01_l"),Weight=0.8),(Item=(Type=Bone,Name="pelvis"),Weight=0.2))
    Weight=@n5.Value
   )
   > &RigUnit_ModifyTransforms(
    ItemToModify="((Item=(Type=Bone,Name=\"MSTR_skirt_side_l\"),Transform=(Rotation=(X=-0.1753,Y=0.0877,Z=-0.2042,W=0.9591),Translation=(X=0,Y=0,Z=0),Scale3D=(X=1,Y=1,Z=1))),(Item=(Type=Bone,Name=\"MSTR_skirt_fr_rear_l\"),Transform=(Rotation=(X=-0.0735,Y=-0.0363,Z=0.0661,W=-0.9 ...(truncated)"
    Weight=@n5.Value
   )
   > &RigUnit_ModifyTransforms(
    ItemToModify="((Item=(Type=Bone,Name=\"base_coatBack_l\"),Transform=(Rotation=(X=0.0288,Y=-0.0192,Z=-0.5338,W=0.8449),Translation=(X=0,Y=0,Z=0),Scale3D=(X=1,Y=1,Z=1))),(Item=(Type=Bone,Name=\"MSTR_skirt_side_rear_l\"),Transform=(Rotation=(X=-0.1335,Y=0.34,Z=-0.3438,W=0.8651 ...(truncated)"
    WeightMinimum=0
    Weight=RigVMVariableNode() {variable=Thigh_L_BackRotation}.Value
   )
   n6 = &RigVMFunction_AlphaInterp(
    InterpSpeedDecreasing=10
    InterpSpeedIncreasing=10
    Scale=1
    bClampResult=true
    Value=&RigVMFunction_MathBoolToFloat(Value=RigVMVariableNode() {variable=Rban}.Value).Result
   )
   Multiply = RigVMAggregateNode(
    C=@n6.Result
    B=&RigVMFunction_MathDoubleRemap(
     SourceMaximum=1
     SourceMinimum=0.2
     TargetMaximum=0
     TargetMinimum=1
     bClamp=False
     Value=RigVMDispatchNode(BoneName=root, Default=1, Name=deformRig_l_thigh_up_alpha) {template="DISPATCH_RigDispatch_GetAnimAttribute(in Name,in BoneName,in Default,out Value,out Found)"}.Value
    ).Result
    A=RigVMVariableNode() {variable=Calf_L_BackRotation}.Value
   )
   > &RigUnit_ModifyTransforms(
    ItemToModify="((Item=(Type=Bone,Name=\"dyn_skirt_side_02_l\"),Transform=(Rotation=(X=-0.062,Y=0.0708,Z=-0.1144,W=0.989),Translation=(X=0,Y=0,Z=0),Scale3D=(X=1,Y=1,Z=1))),(Item=(Type=Bone,Name=\"dyn_coatBack_02_l\"),Transform=(Rotation=(X=0,Y=0,Z=-0.3081,W=0.9514),Translatio ...(truncated)"
    WeightMinimum=0
    Weight=@Multiply.Result
   )
   RigVMFunction_MathBoolToFloat = &RigVMFunction_MathBoolToFloat(Value=RigVMVariableNode() {variable=Rban}.Value)
   AlphaInterp = &RigVMFunction_AlphaInterp(InterpSpeedDecreasing=10, InterpSpeedIncreasing=10, Scale=1, bClampResult=true, Value=@RigVMFunction_MathBoolToFloat.Result)
   Remap = &RigVMFunction_MathDoubleRemap(SourceMaximum=1, SourceMinimum=0, TargetMaximum=0, TargetMinimum=1, bClamp=False, Value=@AlphaInterp.Result)
   > &RigUnit_ModifyTransforms(
    ItemToModify="((Item=(Type=Bone,Name=\"dyn_coatSide_02_l\"),Transform=(Rotation=(X=-0.0622,Y=0.0693,Z=-0.1143,W=0.9891),Translation=(X=0,Y=0,Z=0),Scale3D=(X=1,Y=1,Z=1))),(Item=(Type=Bone,Name=\"dyn_coatBack_02_l\"),Transform=(Rotation=(X=0.0048,Y=0.0528,Z=-0.1899,W=0.9804), ...(truncated)"
    WeightMinimum=0
    Weight=&RigVMFunction_MathDoubleMul(B=@Remap.Result, A=RigVMVariableNode() {variable=Calf_L_BackRotation}.Value).Result
   )
   &RigUnit_ParentConstraint(
    AdvancedSettings=(RotationOrderForFilter=XZY)
    Child=(Type=Bone,Name="base_coatSide_r")
    Parents=((Item=(Type=Bone,Name="thigh_twist_01_r"),Weight=0.25),(Item=(Type=Bone,Name="pelvis"),Weight=0.75))
    Weight=1
    ExecutePin=@RigVMFunction_Sequence.B
   )
   > &RigUnit_ParentConstraint(
    AdvancedSettings=(InterpolationType=Average,RotationOrderForFilter=XZY)
    Child=(Type=Bone,Name="base_coatFrontSide_r")
    Parents=((Item=(Type=Bone,Name="thigh_twist_01_r"),Weight=0.35),(Item=(Type=Bone,Name="pelvis"),Weight=0.65))
    Weight=1
   )
   n7 = RigVMDispatchNode(BoneName=root, Default=1, Name=deformRig_r_thigh_up_alpha) {template="DISPATCH_RigDispatch_GetAnimAttribute(in Name,in BoneName,in Default,out Value,out Found)"}
   > &RigUnit_ModifyTransforms(
    ItemToModify="((Item=(Type=Bone,Name=\"base_coatFrontSide_r\"),Transform=(Rotation=(X=0.5067,Y=-0.0233,Z=0.3429,W=-0.7906),Translation=(X=0,Y=0,Z=0),Scale3D=(X=1,Y=1,Z=1))),(Item=(Type=Bone,Name=\"base_coatSide_r\"),Transform=(Rotation=(X=0.2439,Y=0.0448,Z=0.1596,W=-0.9555) ...(truncated)"
    Weight=@n7.Value
   )
   > &RigUnit_ModifyTransforms(
    ItemToModify="((Item=(Type=Bone,Name=\"base_coatFront_l\"),Transform=(Rotation=(X=0.2605,Y=0.0352,Z=-0.2547,W=0.9306),Translation=(X=-0.5542,Y=-1.4904,Z=-3.9034),Scale3D=(X=1,Y=1,Z=1))),(Item=(Type=Bone,Name=\"base_coatFrontSide_l\"),Transform=(Rotation=(X=-0.2503,Y=0.0106, ...(truncated)"
    Weight=@n7.Value
   )
   n8 = RigVMVariableNode() {variable=Thigh_R_SideRotation}
   > &RigUnit_ParentConstraint(
    AdvancedSettings=(RotationOrderForFilter=XZY)
    Child=(Type=Bone,Name="base_coatSide_r")
    Filter=(TranslationFilter=(bX=false,bY=false,bZ=false),RotationFilter=(bX=True,bY=True,bZ=True),ScaleFilter=(bX=True,bY=True,bZ=True))
    Parents=((Item=(Type=Bone,Name="thigh_twist_01_r"),Weight=0.8),(Item=(Type=Bone,Name="pelvis"),Weight=0.2))
    Weight=@n8.Value
   )
   > &RigUnit_ParentConstraint(
    AdvancedSettings=(RotationOrderForFilter=XZY)
    Child=(Type=Bone,Name="base_coatFrontSide_r")
    Filter=(TranslationFilter=(bX=false,bY=false,bZ=false),RotationFilter=(bX=True,bY=True,bZ=True),ScaleFilter=(bX=True,bY=True,bZ=True))
    Parents=((Item=(Type=Bone,Name="thigh_twist_01_r"),Weight=0.8),(Item=(Type=Bone,Name="pelvis"),Weight=0.2))
    Weight=@n8.Value
   )
   > &RigUnit_ModifyTransforms(
    ItemToModify="((Item=(Type=Bone,Name=\"base_coatSide_r\"),Transform=(Rotation=(X=0.1334,Y=-0.0616,Z=-0.2401,W=0.9596),Translation=(X=0,Y=0,Z=0),Scale3D=(X=1,Y=1,Z=1))),(Item=(Type=Bone,Name=\"base_coatFrontSide_r\"),Transform=(Rotation=(X=0.0072,Y=-0.1662,Z=0.1719,W=-0.971) ...(truncated)"
    Weight=@n8.Value
   )
   > &RigUnit_ModifyTransforms(
    ItemToModify="((Item=(Type=Bone,Name=\"base_coatBack_r\"),Transform=(Rotation=(X=0.0288,Y=-0.0192,Z=-0.5338,W=0.8449),Translation=(X=0,Y=0,Z=0),Scale3D=(X=1,Y=1,Z=1))),(Item=(Type=Bone,Name=\"base_coatBackSide_r\"),Transform=(Rotation=(X=0.3279,Y=-0.3457,Z=-0.2928,W=0.829), ...(truncated)"
    WeightMinimum=0
    Weight=RigVMVariableNode() {variable=Thigh_R_BackRotation}.Value
   )
   n9 = &RigVMFunction_AlphaInterp(
    InterpSpeedDecreasing=10
    InterpSpeedIncreasing=10
    Scale=1
    bClampResult=true
    Value=&RigVMFunction_MathBoolToFloat(Value=RigVMVariableNode() {variable=Rban}.Value).Result
   )
   Multiply_2 = RigVMAggregateNode(
    C=@n9.Result
    B=&RigVMFunction_MathDoubleRemap(
     SourceMaximum=1
     SourceMinimum=0.2
     TargetMaximum=0
     TargetMinimum=1
     bClamp=False
     Value=RigVMDispatchNode(BoneName=root, Default=1, Name=deformRig_r_thigh_up_alpha) {template="DISPATCH_RigDispatch_GetAnimAttribute(in Name,in BoneName,in Default,out Value,out Found)"}.Value
    ).Result
    A=RigVMVariableNode() {variable=Calf_R_BackRotation}.Value
   )
   > &RigUnit_ModifyTransforms(
    ItemToModify="((Item=(Type=Bone,Name=\"dyn_coatBack_03_r\"),Transform=(Rotation=(X=0.0071,Y=0.0141,Z=0.2527,W=-0.9674),Translation=(X=0,Y=0,Z=0),Scale3D=(X=1,Y=1,Z=1))),(Item=(Type=Bone,Name=\"dyn_coatSideBack_03_r\"),Transform=(Rotation=(X=-0.0272,Y=-0.1393,Z=0.1524,W=-0.9 ...(truncated)"
    WeightMinimum=0
    Weight=@Multiply_2.Result
   )
   n10 = &RigVMFunction_AlphaInterp(
    InterpSpeedDecreasing=10
    InterpSpeedIncreasing=10
    Scale=1
    bClampResult=true
    Value=&RigVMFunction_MathBoolToFloat(Value=RigVMVariableNode() {variable=Rban}.Value).Result
   )
   > &RigUnit_ModifyTransforms(
    ItemToModify="((Item=(Type=Bone,Name=\"dyn_coatBack_03_r\"),Transform=(Rotation=(X=0.0086,Y=0.0132,Z=0.1401,W=-0.99),Translation=(X=0,Y=0,Z=0),Scale3D=(X=1,Y=1,Z=1))),(Item=(Type=Bone,Name=\"dyn_coatSideBack_03_r\"),Transform=(Rotation=(X=-0.0272,Y=-0.1393,Z=0.1524,W=-0.978 ...(truncated)"
    WeightMinimum=0
    Weight=&RigVMFunction_MathDoubleMul(
     B=&RigVMFunction_MathDoubleRemap(SourceMaximum=1, SourceMinimum=0, TargetMaximum=0, TargetMinimum=1, bClamp=False, Value=@n10.Result).Result
     A=RigVMVariableNode() {variable=Calf_R_BackRotation}.Value
    ).Result
   )
   n11 = RigVMDispatchNode(BoneName=root, Default=1, Name=deformRig_l_thigh_up_alpha) {template="DISPATCH_RigDispatch_GetAnimAttribute(in Name,in BoneName,in Default,out Value,out Found)"}
   &RigUnit_ParentConstraint(
    AdvancedSettings=(RotationOrderForFilter=XZY)
    Child=(Type=Bone,Name="MSTR_skirt_fr_l")
    Filter=(TranslationFilter=(bX=false,bY=false,bZ=false),RotationFilter=(bX=True,bY=True,bZ=True),ScaleFilter=(bX=True,bY=True,bZ=True))
    Parents=((Item=(Type=Bone,Name="thigh_twist_01_l"),Weight=0.55),(Item=(Type=Bone,Name="pelvis"),Weight=0.45))
    Weight=@n11.Value
    ExecutePin=@RigVMFunction_Sequence.C
   )
   > &RigUnit_ModifyTransforms(
    ItemToModify="((Item=(Type=Bone,Name=\"None\"),Transform=(Rotation=(X=0,Y=0,Z=0,W=1),Translation=(X=0,Y=0,Z=0),Scale3D=(X=1,Y=1,Z=1))),(Item=(Type=Bone,Name=\"MSTR_skirt_fr_l\"),Transform=(Rotation=(X=-0.2408,Y=-0.1166,Z=0.2983,W=-0.9162),Translation=(X=0,Y=0,Z=0),Scale3D=( ...(truncated)"
    WeightMinimum=0
    Weight=@n11.Value
   )
   > &RigUnit_ModifyTransforms(
    ItemToModify="((Item=(Type=Bone,Name=\"None\"),Transform=(Rotation=(X=0,Y=0,Z=0,W=1),Translation=(X=0,Y=0,Z=0),Scale3D=(X=1,Y=1,Z=1))),(Item=(Type=Bone,Name=\"MSTR_skirt_fr_l\"),Transform=(Rotation=(X=0.0551,Y=-0.3665,Z=0.1381,W=-0.9185),Translation=(X=0,Y=0,Z=0),Scale3D=(X ...(truncated)"
    WeightMinimum=0
    Weight=RigVMVariableNode() {variable=Thigh_L_SideRotation}.Value
   )
   n12 = RigVMDispatchNode(BoneName=root, Default=1, Name=deformRig_r_thigh_up_alpha) {template="DISPATCH_RigDispatch_GetAnimAttribute(in Name,in BoneName,in Default,out Value,out Found)"}
   &RigUnit_ParentConstraint(
    AdvancedSettings=(RotationOrderForFilter=XZY)
    Child=(Type=Bone,Name="MSTR_skirt_fr_r")
    Filter=(TranslationFilter=(bX=false,bY=false,bZ=false),RotationFilter=(bX=True,bY=True,bZ=True),ScaleFilter=(bX=True,bY=True,bZ=True))
    Parents=((Item=(Type=Bone,Name="thigh_twist_01_r"),Weight=0.7),(Item=(Type=Bone,Name="pelvis"),Weight=0.3))
    Weight=@n12.Value
    ExecutePin=@RigVMFunction_Sequence.D
   )
   > &RigUnit_ModifyTransforms(
    ItemToModify="((Item=(Type=Bone,Name=\"None\"),Transform=(Rotation=(X=0,Y=0,Z=0,W=1),Translation=(X=0,Y=0,Z=0),Scale3D=(X=1,Y=1,Z=1))),(Item=(Type=Bone,Name=\"MSTR_skirt_fr_r\"),Transform=(Rotation=(X=0.0401,Y=-0.2373,Z=-0.2498,W=0.9379),Translation=(X=0.0421,Y=0.0944,Z=0.0 ...(truncated)"
    WeightMinimum=0
    Weight=@n12.Value
   )
   > &RigUnit_ModifyTransforms(
    ItemToModify="((Item=(Type=Bone,Name=\"None\"),Transform=(Rotation=(X=0,Y=0,Z=0,W=1),Translation=(X=0,Y=0,Z=0),Scale3D=(X=1,Y=1,Z=1))),(Item=(Type=Bone,Name=\"MSTR_skirt_fr_r\"),Transform=(Rotation=(X=-0.0518,Y=-0.3371,Z=-0.2547,W=0.9049),Translation=(X=0,Y=0,Z=0),Scale3D=( ...(truncated)"
    WeightMinimum=0
    Weight=RigVMVariableNode() {variable=Thigh_R_SideRotation}.Value
   )
   n13 = &RigVMFunction_AlphaInterp(
    InterpSpeedDecreasing=10
    InterpSpeedIncreasing=10
    OutRange=(Min=0,Max=1.4)
    Scale=1
    bClampResult=true
    Value=&RigVMFunction_MathBoolToFloat(Value=RigVMVariableNode() {variable=isCrouching_CR}.Value).Result
   )
   &RigUnit_ModifyTransforms(
    ItemToModify="((Item=(Type=Bone,Name=\"None\"),Transform=(Rotation=(X=0,Y=0,Z=0,W=1),Translation=(X=0,Y=0,Z=0),Scale3D=(X=1,Y=1,Z=1))),(Item=(Type=Bone,Name=\"base_coatFrontSide_l\"),Transform=(Rotation=(X=0.1777,Y=-0.1691,Z=0.0236,W=-0.9691),Translation=(X=0.1254,Y=-1.7477 ...(truncated)"
    WeightMinimum=0
    Weight=@n13.Result
    ExecutePin=@RigVMFunction_Sequence.E
   )
   > &RigUnit_ModifyTransforms(
    ItemToModify="((Item=(Type=Bone,Name=\"base_coatFront_r\"),Transform=(Rotation=(X=0,Y=0,Z=-0.0297,W=0.9996),Translation=(X=-3.9325,Y=0,Z=0),Scale3D=(X=1,Y=1,Z=1))),(Item=(Type=Bone,Name=\"base_coatFront_l\"),Transform=(Rotation=(X=-0.0395,Y=-0.0029,Z=0.0255,W=0.9989),Transl ...(truncated)"
    Weight=&RigVMFunction_MathDoubleMul(A=@n13.Result, B=RigVMVariableNode() {variable=spineTwist_R}.Value).Result
   )
   n14 = &RigVMFunction_AlphaInterp(
    InterpSpeedDecreasing=10
    InterpSpeedIncreasing=10
    Scale=1
    bClampResult=true
    Value=&RigVMFunction_MathBoolToFloat(Value=RigVMVariableNode() {variable=isBaseballSliding_CR}.Value).Result
   )
   > &RigUnit_ModifyTransforms(
    ItemToModify="((Item=(Type=Bone,Name=\"None\"),Transform=(Rotation=(X=0,Y=0,Z=0,W=1),Translation=(X=0,Y=0,Z=0),Scale3D=(X=1,Y=1,Z=1))),(Item=(Type=Bone,Name=\"base_coatFront_r\"),Transform=(Rotation=(X=0.0408,Y=-0.0045,Z=0.0209,W=0.9989),Translation=(X=0.3819,Y=-9.644,Z=-1. ...(truncated)"
    WeightMinimum=0
    Weight=@n14.Result
   )
   > &RigUnit_ModifyTransforms(
    ItemToModify="((Item=(Type=Bone,Name=\"None\"),Transform=(Rotation=(X=0,Y=0,Z=0,W=1),Translation=(X=0,Y=0,Z=0),Scale3D=(X=1,Y=1,Z=1))),(Item=(Type=Bone,Name=\"dyn_coatBack_02_l\"),Transform=(Rotation=(X=0,Y=0,Z=-0.2493,W=-0.9684),Translation=(X=0,Y=0,Z=0),Scale3D=(X=1,Y=1,Z ...(truncated)"
    WeightMinimum=0
    Weight=@n14.Result
   )
   n15 = &RigVMFunction_AlphaInterp(
    InterpSpeedDecreasing=10
    InterpSpeedIncreasing=10
    Scale=1
    bClampResult=true
    Value=&RigVMFunction_MathBoolToFloat(Value=RigVMVariableNode() {variable=isKneeSliding_CR}.Value).Result
   )
   > &RigUnit_ModifyTransforms(
    ItemToModify="((Item=(Type=Bone,Name=\"None\"),Transform=(Rotation=(X=0,Y=0,Z=0,W=1),Translation=(X=0,Y=0,Z=0),Scale3D=(X=1,Y=1,Z=1))),(Item=(Type=Bone,Name=\"base_coatFront_l\"),Transform=(Rotation=(X=0.0078,Y=0.1672,Z=0.083,W=-0.9824),Translation=(X=0,Y=0,Z=0),Scale3D=(X= ...(truncated)"
    WeightMinimum=0
    Weight=@n15.Result
   )
   Multiply_9 = RigVMAggregateNode(A=1, B=RigVMVariableNode() {variable=TacticalSprintingNoRB}.Value, C=RigVMVariableNode() {variable=Calf_L_BackRotation}.Value)
   &RigUnit_ModifyTransforms(
    ExecutePin=()
    ItemToModify="((Item=(Type=Bone,Name=\"dyn_coatBack_02_l\"),Transform=(Rotation=(X=-0.0538,Y=-0.1119,Z=-0.4296,W=0.8944),Translation=(X=0,Y=0,Z=0),Scale3D=(X=1,Y=1,Z=1))),(Item=(Type=Bone,Name=\"dyn_coatBack_03_l\"),Transform=(Rotation=(X=-0.0389,Y=-0.1116,Z=-0.1026,W=0.987 ...(truncated)"
    WeightMinimum=0
    Weight=@Multiply_9.Result
   )
   n16 = &RigVMFunction_AlphaInterp(InterpSpeedDecreasing=1, InterpSpeedIncreasing=10, Scale=1, bClampResult=true, Value=RigVMVariableNode() {variable=jumpsprint}.Value)
   > &RigUnit_ModifyTransforms(
    ItemToModify=((Item=(Type=Bone,Name="base_coatBack_l"),Transform=(Rotation=(X=0.0235,Y=0.1352,Z=0.2515,W=-0.9581),Translation=(X=0,Y=0,Z=0),Scale3D=(X=1,Y=1,Z=1))))
    WeightMinimum=0
    Weight=@n16.Result
   )
   Multiply_11 = RigVMAggregateNode(A=1, C=RigVMVariableNode() {variable=Calf_R_BackRotation}.Value, B=RigVMVariableNode() {variable=TacticalSprintingNoRB}.Value)
   &RigUnit_ModifyTransforms(
    ExecutePin=()
    ItemToModify="((Item=(Type=Bone,Name=\"dyn_coatBack_02_r\"),Transform=(Rotation=(X=0.0484,Y=0.0864,Z=-0.4865,W=0.8681),Translation=(X=0,Y=0,Z=0),Scale3D=(X=1,Y=1,Z=1))),(Item=(Type=Bone,Name=\"dyn_coatBack_03_r\"),Transform=(Rotation=(X=-0.0063,Y=0.003,Z=0.0851,W=-0.9964),T ...(truncated)"
    WeightMinimum=0
    Weight=@Multiply_11.Result
   )
   n17 = &RigVMFunction_AlphaInterp(
    InterpSpeedDecreasing=10
    InterpSpeedIncreasing=10
    OutRange=(Min=0,Max=1.4)
    Scale=1
    bClampResult=true
    Value=&RigVMFunction_MathBoolToFloat(Value=RigVMVariableNode() {variable=isCrouching_CR}.Value).Result
   )
   &RigUnit_ModifyTransforms(
    ExecutePin=()
    ItemToModify="((Item=(Type=Bone,Name=\"base_coatFront_l\"),Transform=(Rotation=(X=-0.0483,Y=0.0002,Z=-0.5381,W=0.8415),Translation=(X=-6.2979,Y=-1.7869,Z=0.4762),Scale3D=(X=1,Y=1,Z=1))),(Item=(Type=Bone,Name=\"base_coatFrontSide_r\"),Transform=(Rotation=(X=0,Y=0,Z=0,W=1),Tr ...(truncated)"
    Weight=&RigVMFunction_MathDoubleMul(A=1, B=@n17.Result).Result
   )
  ]}
 RigVMFunctionLibrary {nodes [
   Coat = RigVMCollapseNode(ExecuteContext=())
   Control Setup = RigVMCollapseNode(Debug Controls=False, ExecuteContext=())
   Delta Rotation X = RigVMCollapseNode(ExecuteContext=())
   Delta Rotation Y = RigVMCollapseNode(ExecuteContext=())
   Delta Rotation Z = RigVMCollapseNode(ExecuteContext=())
   Variables = RigVMCollapseNode(ExecuteContext=())
  ]}
 RigVMFunction_Sequence_1_ContainedGraph {nodes [
   n1 = RigVMFunctionEntryNode()
   RigVMFunction_Sequence = &RigVMFunction_Sequence(ExecuteContext=@n1.ExecuteContext)
   n2 = &RigVMFunction_Sequence(ExecuteContext=@RigVMFunction_Sequence.B)
   n3 = &RigVMFunction_Sequence(ExecuteContext=@n2.B)
   RigVMFunctionReturnNode(A=@RigVMFunction_Sequence.A, B=@n2.A, C=@n3.A, D=@n3.B)
  ]}
 RigVMFunction_Sequence_1_ContainedGraph {nodes [
   n1 = RigVMFunctionEntryNode()
   RigVMFunction_Sequence = &RigVMFunction_Sequence(ExecuteContext=@n1.ExecuteContext)
   n2 = &RigVMFunction_Sequence(ExecuteContext=@RigVMFunction_Sequence.B)
   RigVMFunctionReturnNode(A=@RigVMFunction_Sequence.A, B=@n2.A, C=@n2.B)
  ]}
 RigVMFunction_Sequence_1_ContainedGraph {nodes [
   n1 = RigVMFunctionEntryNode()
   RigVMFunction_Sequence = &RigVMFunction_Sequence(ExecuteContext=@n1.ExecuteContext)
   n2 = &RigVMFunction_Sequence(ExecuteContext=@RigVMFunction_Sequence.B)
   n3 = &RigVMFunction_Sequence(ExecuteContext=@n2.B)
   n4 = &RigVMFunction_Sequence(ExecuteContext=@n3.B)
   n5 = &RigVMFunction_Sequence(ExecuteContext=@n4.B)
   n6 = &RigVMFunction_Sequence(ExecuteContext=@n5.B)
   n7 = &RigVMFunction_Sequence(ExecuteContext=@n6.B)
   RigVMFunctionReturnNode(A=@RigVMFunction_Sequence.A, B=@n2.A, C=@n3.A, D=@n4.A, E=@n5.A, F=@n6.A, G=@n7.A, H=@n7.B)
  ]}
 RigVMModel {nodes [
   RigUnit_BeginExecution = &RigUnit_BeginExecution()
   /Script/FortniteGame.RigUnit_FortDeformRig(
    Bool=(bUpperArm=true,bElbow=true,bWrist=true,bGroin=true,bGlute=true,bKnee=true,bPec=true,bBicep=False,bLat=False)
    Curve="(bLUpperArmUpAlpha=true,bRUpperArmUpAlpha=true,bLUpperArmFrAlpha=true,bRUpperArmFrAlpha=true,bLUpperArmBkAlpha=true,bRUpperArmBkAlpha=true,bLWristUpAlpha=False,bRWristUpAlpha=False,bLWristDnAlpha=False,bRWristDnAlpha=False,bLKneeAlpha=False,bRKneeAlpha=Fal ...(truncated)"
    Debug=(bUpperArmDebug=False,bWristDebug=False,bLatDebug=False,bGroinGluteDebug=False)
    Value="(UpperArmUp=(Rotation=(X=0.0,Y=-0.0,Z=0.0,W=1.0),Translation=(X=0.0,Y=0.0,Z=8.0),Scale3D=(X=0.70,Y=1.0,Z=1.0)),UpperArmFwdFr=(Rotation=(X=0.0,Y=-0.0,Z=0.0,W=1.0),Translation=(X=0.0,Y=10.0,Z=0.0),Scale3D=(X=0.50,Y=1.0,Z=1.0)),UpperArmFwdBk=(Rotation=(X=0.0, ...(truncated)"
    ExecutePin=@RigUnit_BeginExecution.ExecutePin
   ) {is=struct RigVMUnitNode}
   > RigVMFunction_Sequence = RigVMAggregateNode()
   Leg ROM = RigVMFunctionReferenceNode(
    Include Ankles=true
    Include Distance Checks=true
    Include Knees=true
    Interp Output Speed=5
    Interp Results=true
    Result Alpha=1
    SPR Optional Parent Item=(Type=Bone,Name="None")
    ExecuteContext=@RigVMFunction_Sequence.A
   )
   > RigVMVariableNode(Value=@"Leg ROM".L Thigh Rear Knee Bend) {variable=Leg ROM_L Thigh Rear Knee Bend}
   > RigVMVariableNode(Value=@"Leg ROM".R Thigh Rear Knee Bend) {variable=Leg ROM_R Thigh Rear Knee Bend}
   > RigVMVariableNode(Value=@"Leg ROM".L Thigh Raised Fwd) {variable=LegForward_L}
   > RigVMVariableNode(Value=@"Leg ROM".R Thigh Raised Fwd) {variable=LegForward_R}
   > RigVMVariableNode(Value=@"Leg ROM".L Thigh Raised Fwd Side Inbetween) {variable=Leg ROM_L Thigh Raised Fwd Side Inbetween}
   > RigVMVariableNode(Value=@"Leg ROM".R Thigh Raised Fwd Side Inbetween) {variable=Leg ROM_R Thigh Raised Fwd Side Inbetween}
   &RigUnit_ParentConstraint(
    AdvancedSettings=(InterpolationType=Average,RotationOrderForFilter=XZY)
    Child=(Type=Control,Name="pelvis_ctrl")
    Filter=(TranslationFilter=(bX=true,bY=true,bZ=true),RotationFilter=(bX=false,bY=false,bZ=false),ScaleFilter=(bX=True,bY=True,bZ=True))
    Parents=((Item=(Type=Bone,Name="thigh_l"),Weight=1),(Item=(Type=Bone,Name="thigh_r"),Weight=1))
    Weight=1
    ExecutePin=@RigVMFunction_Sequence.B
   )
   > /Script/ControlRig.RigUnit_AimItem(
    DebugSettings=(bEnabled=False,Scale=10,WorldOffset=(Rotation=(X=0,Y=0,Z=0,W=1),Translation=(X=0,Y=0,Z=0),Scale3D=(X=1,Y=1,Z=1)))
    Item=(Type=Bone,Name="def_twist_corr_l")
    Primary=(Weight=1,Axis=(X=-1,Y=0,Z=0),Target=(X=0,Y=-1,Z=0.5),Kind=Location,Space=(Type=Bone,Name="calf_l"))
    Secondary=(Weight=1,Axis=(X=0,Y=0,Z=1),Target=(X=0,Y=0,Z=1),Kind=Location,Space=(Type=Control,Name="pelvis_ctrl"))
    Weight=1
   ) {is=struct RigVMUnitNode}
   > &RigUnit_AimItem(
    DebugSettings=(bEnabled=False,Scale=10,WorldOffset=(Rotation=(X=0,Y=0,Z=0,W=1),Translation=(X=0,Y=0,Z=0),Scale3D=(X=1,Y=1,Z=1)))
    Item=(Type=Bone,Name="def_twist_corr_r")
    Primary=(Weight=1,Axis=(X=1,Y=0,Z=0),Target=(X=0,Y=1,Z=-0.5),Kind=Location,Space=(Type=Bone,Name="calf_r"))
    Secondary=(Weight=1,Axis=(X=0,Y=0,Z=-1),Target=(X=0,Y=0,Z=-1),Kind=Location,Space=(Type=Control,Name="pelvis_ctrl"))
    Weight=1
   )
   n1 = &RigVMFunction_MathQuaternionToEuler(
    RotationOrder=ZYX
    Value=&RigUnit_GetTransform(Item=(Type=Bone,Name="thigh_l"), Space=LocalSpace, bInitial=False).Transform.Rotation
   )
   RigVMFunction_MathQuaternionFromEuler = &RigVMFunction_MathQuaternionFromEuler(Euler=(X=0,Y=0,Z=0), Euler.X=&RigVMFunction_MathDoubleMul(B=0.5, A=@n1.Result.X).Result)
   > &RigUnit_ModifyTransforms(
    ItemToModify=((Item=(Type=Bone,Name="def_twist_corr_l"),Transform=(Rotation=(X=0,Y=0,Z=0,W=1),Translation=(X=0,Y=0,Z=0.5),Scale3D=(X=1,Y=0.9,Z=0.7))))
    Weight=1
    WeightMinimum=0
    ItemToModify.0.Transform.Scale3D.Z=&RigVMFunction_MathDoubleRemap(SourceMaximum=-20, SourceMinimum=-5, TargetMaximum=0.7, TargetMinimum=0.95, Value=@n1.Result.X).Result
    ItemToModify.0.Transform.Rotation=@RigVMFunction_MathQuaternionFromEuler.Result
   )
   n2 = &RigVMFunction_MathQuaternionToEuler(
    RotationOrder=ZYX
    Value=&RigUnit_GetTransform(Item=(Type=Bone,Name="thigh_r"), Space=LocalSpace, bInitial=False).Transform.Rotation
   )
   > &RigUnit_ModifyTransforms(
    ItemToModify=((Item=(Type=Bone,Name="def_twist_corr_r"),Transform=(Rotation=(X=0,Y=0,Z=0,W=1),Translation=(X=0,Y=0,Z=-0.5),Scale3D=(X=1,Y=0.9,Z=1))))
    Weight=1
    WeightMinimum=0
    ItemToModify.0.Transform.Scale3D.Z=&RigVMFunction_MathDoubleRemap(SourceMaximum=20, SourceMinimum=5, TargetMaximum=0.7, TargetMinimum=0.95, Value=@n2.Result.X).Result
    ItemToModify.0.Transform.Rotation=&RigVMFunction_MathQuaternionFromEuler(Euler=(X=0,Y=0,Z=0), Euler.X=&RigVMFunction_MathDoubleMul(B=-0.5, A=@n2.Result.X).Result).Result
   )
   n3 = &RigUnit_GetTransform(Item=(Type=Bone,Name="calf_r"), Space=LocalSpace, bInitial=False)
   > &RigUnit_ModifyTransforms(
    ItemToModify="((Item=(Type=Bone,Name=\"def_twist_corr_r\"),Transform=(Rotation=(X=0,Y=0,Z=0,W=1),Translation=(X=-2,Y=0,Z=0),Scale3D=(X=0.9,Y=0.9,Z=0.9))),(Item=(Type=Bone,Name=\"def_twist_corr_bk_r\"),Transform=(Rotation=(X=0,Y=0,Z=0,W=1),Translation=(X=-3,Y=0,Z=0),Scale3D= ...(truncated)"
    WeightMinimum=0
    Weight=&RigVMFunction_MathDoubleRemap(
     SourceMaximum=110
     SourceMinimum=60
     TargetMaximum=1
     TargetMinimum=0
     Value=&RigVMFunction_MathQuaternionToEuler(RotationOrder=ZYX, Value=@n3.Transform.Rotation).Result.Z
    ).Result
   )
   n4 = &RigUnit_GetTransform(Item=(Type=Bone,Name="calf_l"), Space=LocalSpace, bInitial=False)
   > &RigUnit_ModifyTransforms(
    ItemToModify="((Item=(Type=Bone,Name=\"def_twist_corr_l\"),Transform=(Rotation=(X=0,Y=0,Z=0,W=1),Translation=(X=2,Y=0,Z=0),Scale3D=(X=0.9,Y=0.9,Z=0.9))),(Item=(Type=Bone,Name=\"def_twist_corr_bk_l\"),Transform=(Rotation=(X=-0.3407,Y=0.0298,Z=-0.0819,W=0.9361),Translation=(X ...(truncated)"
    WeightMinimum=0
    Weight=&RigVMFunction_MathDoubleRemap(
     SourceMaximum=110
     SourceMinimum=60
     TargetMaximum=1
     TargetMinimum=0
     Value=&RigVMFunction_MathQuaternionToEuler(RotationOrder=ZYX, Value=@n4.Transform.Rotation).Result.Z
    ).Result
   )
   &RigUnit_ParentConstraint(
    AdvancedSettings=(InterpolationType=Average,RotationOrderForFilter=XZY)
    Child=(Type=Bone,Name="MSTR_skirt_side_l")
    Filter=(TranslationFilter=(bX=false,bY=false,bZ=false),RotationFilter=(bX=true,bY=true,bZ=true),ScaleFilter=(bX=true,bY=true,bZ=true))
    Parents=((Item=(Type=Bone,Name="thigh_twist_01_l"),Weight=0.25),(Item=(Type=Bone,Name="pelvis"),Weight=1))
    Weight=1
    ExecutePin=@RigVMFunction_Sequence.C
   )
   > &RigUnit_ParentConstraint(
    AdvancedSettings=(InterpolationType=Average,RotationOrderForFilter=XZY)
    Child=(Type=Bone,Name="MSTR_skirt_side_rear_l")
    Filter=(TranslationFilter=(bX=false,bY=false,bZ=false),RotationFilter=(bX=true,bY=true,bZ=true),ScaleFilter=(bX=true,bY=true,bZ=true))
    Parents=((Item=(Type=Bone,Name="thigh_twist_01_l"),Weight=0.25),(Item=(Type=Bone,Name="pelvis"),Weight=1))
    Weight=1
   )
   > &RigUnit_ParentConstraint(
    AdvancedSettings=(InterpolationType=Average,RotationOrderForFilter=XZY)
    Child=(Type=Bone,Name="MSTR_skirt_bk_l")
    Filter=(TranslationFilter=(bX=false,bY=false,bZ=false),RotationFilter=(bX=true,bY=true,bZ=true),ScaleFilter=(bX=true,bY=true,bZ=true))
    Parents=((Item=(Type=Bone,Name="thigh_l"),Weight=0.1),(Item=(Type=Bone,Name="thigh_r"),Weight=0),(Item=(Type=Bone,Name="pelvis"),Weight=0.9))
    Weight=1
   )
   > &RigUnit_ParentConstraint(
    AdvancedSettings=(InterpolationType=Average,RotationOrderForFilter=XZY)
    Child=(Type=Bone,Name="MSTR_skirt_fr_l")
    Filter=(TranslationFilter=(bX=false,bY=false,bZ=false),RotationFilter=(bX=true,bY=true,bZ=true),ScaleFilter=(bX=true,bY=true,bZ=true))
    Parents=((Item=(Type=Bone,Name="thigh_l"),Weight=1),(Item=(Type=Bone,Name="pelvis"),Weight=0),(Item=(Type=Bone,Name="deform_groin_l"),Weight=0))
    Weight=1
   )
   > &RigUnit_ParentConstraint(
    AdvancedSettings=(InterpolationType=Average,RotationOrderForFilter=XZY)
    Child=(Type=Bone,Name="MSTR_skirt_fr_l")
    Filter=(TranslationFilter=(bX=true,bY=true,bZ=true),RotationFilter=(bX=false,bY=false,bZ=false),ScaleFilter=(bX=false,bY=false,bZ=false))
    Parents=((Item=(Type=Bone,Name="deform_groin_l"),Weight=1),(Item=(Type=Bone,Name="def_twist_corr_l"),Weight=1))
    Weight=1
   )
   > &RigUnit_ParentConstraint(
    AdvancedSettings=(InterpolationType=Average,RotationOrderForFilter=XZY)
    Child=(Type=Bone,Name="MSTR_skirt_fr_rear_l")
    Filter=(TranslationFilter=(bX=false,bY=false,bZ=false),RotationFilter=(bX=true,bY=true,bZ=true),ScaleFilter=(bX=true,bY=true,bZ=true))
    Parents=((Item=(Type=Bone,Name="thigh_l"),Weight=1),(Item=(Type=Bone,Name="pelvis"),Weight=0))
    Weight=1
   )
   > &RigUnit_ParentConstraint(
    AdvancedSettings=(InterpolationType=Average,RotationOrderForFilter=XZY)
    Child=(Type=Bone,Name="MSTR_skirt_fr_rear_l")
    Filter=(TranslationFilter=(bX=true,bY=true,bZ=true),RotationFilter=(bX=false,bY=false,bZ=false),ScaleFilter=(bX=false,bY=false,bZ=false))
    Parents=((Item=(Type=Bone,Name="deform_groin_l"),Weight=1),(Item=(Type=Bone,Name="def_twist_corr_l"),Weight=1))
    Weight=1
   )
   > &RigUnit_ParentConstraint(
    AdvancedSettings=(InterpolationType=Average,RotationOrderForFilter=XZY)
    Child=(Type=Bone,Name="MSTR_skirt_side_r")
    Filter=(TranslationFilter=(bX=false,bY=false,bZ=false),RotationFilter=(bX=true,bY=true,bZ=true),ScaleFilter=(bX=true,bY=true,bZ=true))
    Parents=((Item=(Type=Bone,Name="thigh_twist_01_r"),Weight=0.25),(Item=(Type=Bone,Name="pelvis"),Weight=1))
    Weight=1
   )
   > &RigUnit_ParentConstraint(
    AdvancedSettings=(InterpolationType=Average,RotationOrderForFilter=XZY)
    Child=(Type=Bone,Name="MSTR_skirt_side_rear_r")
    Filter=(TranslationFilter=(bX=false,bY=false,bZ=false),RotationFilter=(bX=true,bY=true,bZ=true),ScaleFilter=(bX=true,bY=true,bZ=true))
    Parents=((Item=(Type=Bone,Name="thigh_twist_01_r"),Weight=0.25),(Item=(Type=Bone,Name="pelvis"),Weight=1))
    Weight=1
   )
   > &RigUnit_ParentConstraint(
    AdvancedSettings=(InterpolationType=Average,RotationOrderForFilter=XZY)
    Child=(Type=Bone,Name="MSTR_skirt_bk_r")
    Filter=(TranslationFilter=(bX=false,bY=false,bZ=false),RotationFilter=(bX=true,bY=true,bZ=true),ScaleFilter=(bX=true,bY=true,bZ=true))
    Parents=((Item=(Type=Bone,Name="thigh_r"),Weight=0.1),(Item=(Type=Bone,Name="thigh_r"),Weight=0),(Item=(Type=Bone,Name="pelvis"),Weight=0.9))
    Weight=1
   )
   > &RigUnit_ParentConstraint(
    AdvancedSettings=(InterpolationType=Average,RotationOrderForFilter=XZY)
    Child=(Type=Bone,Name="MSTR_skirt_fr_r")
    Filter=(TranslationFilter=(bX=false,bY=false,bZ=false),RotationFilter=(bX=true,bY=true,bZ=true),ScaleFilter=(bX=true,bY=true,bZ=true))
    Parents=((Item=(Type=Bone,Name="thigh_r"),Weight=1),(Item=(Type=Bone,Name="pelvis"),Weight=0),(Item=(Type=Bone,Name="deform_groin_r"),Weight=0))
    Weight=1
   )
   > &RigUnit_ParentConstraint(
    AdvancedSettings=(InterpolationType=Average,RotationOrderForFilter=XZY)
    Child=(Type=Bone,Name="MSTR_skirt_fr_r")
    Filter=(TranslationFilter=(bX=true,bY=true,bZ=true),RotationFilter=(bX=false,bY=false,bZ=false),ScaleFilter=(bX=false,bY=false,bZ=false))
    Parents=((Item=(Type=Bone,Name="deform_groin_r"),Weight=1),(Item=(Type=Bone,Name="def_twist_corr_r"),Weight=1))
    Weight=1
   )
   > &RigUnit_ParentConstraint(
    AdvancedSettings=(InterpolationType=Average,RotationOrderForFilter=XZY)
    Child=(Type=Bone,Name="MSTR_skirt_fr_rear_r")
    Filter=(TranslationFilter=(bX=false,bY=false,bZ=false),RotationFilter=(bX=true,bY=true,bZ=true),ScaleFilter=(bX=true,bY=true,bZ=true))
    Parents=((Item=(Type=Bone,Name="thigh_r"),Weight=1),(Item=(Type=Bone,Name="pelvis"),Weight=0))
    Weight=1
   )
   > &RigUnit_ParentConstraint(
    AdvancedSettings=(InterpolationType=Average,RotationOrderForFilter=XZY)
    Child=(Type=Bone,Name="MSTR_skirt_fr_rear_r")
    Filter=(TranslationFilter=(bX=true,bY=true,bZ=true),RotationFilter=(bX=false,bY=false,bZ=false),ScaleFilter=(bX=false,bY=false,bZ=false))
    Parents=((Item=(Type=Bone,Name="deform_groin_r"),Weight=1),(Item=(Type=Bone,Name="def_twist_corr_r"),Weight=1))
    Weight=1
   )
   > &RigUnit_ParentConstraint(
    AdvancedSettings=(InterpolationType=Average,RotationOrderForFilter=XZY)
    Child=(Type=Bone,Name="def_alt_skirt_corr_r")
    Filter=(TranslationFilter=(bX=true,bY=true,bZ=true),RotationFilter=(bX=true,bY=true,bZ=true),ScaleFilter=(bX=True,bY=True,bZ=True))
    Parents=((Item=(Type=Bone,Name="thigh_r"),Weight=1),(Item=(Type=Bone,Name="def_twist_corr_r"),Weight=0.1))
    Weight=1
   )
   > &RigUnit_ParentConstraint(
    AdvancedSettings=(InterpolationType=Average,RotationOrderForFilter=XYZ)
    Child=(Type=Bone,Name="def_alt_skirt_corr_l")
    Filter=(TranslationFilter=(bX=true,bY=true,bZ=true),RotationFilter=(bX=true,bY=true,bZ=true),ScaleFilter=(bX=True,bY=True,bZ=True))
    Parents=((Item=(Type=Bone,Name="thigh_l"),Weight=1),(Item=(Type=Bone,Name="def_twist_corr_l"),Weight=0.1))
    Weight=1
   )
   n5 = RigVMVariableNode() {variable=isIdle}
   AlphaInterp = &RigVMFunction_AlphaInterp(
    InterpSpeedDecreasing=10
    InterpSpeedIncreasing=10
    OutRange=(Min=0.3,Max=1)
    Scale=1
    bClampResult=False
    bMapRange=true
    Value=&RigVMFunction_MathBoolToFloat(Value=&RigVMFunction_MathBoolNot(Value=@n5.Value).Result).Result
   )
   > &RigUnit_ModifyTransforms(
    ItemToModify="((Item=(Type=Bone,Name=\"MSTR_skirt_fr_l\"),Transform=(Rotation=(X=0.0872,Y=0,Z=0,W=0.9962),Translation=(X=0,Y=-1,Z=0),Scale3D=(X=1,Y=1,Z=1))),(Item=(Type=Bone,Name=\"MSTR_skirt_fr_rear_l\"),Transform=(Rotation=(X=0.1305,Y=0,Z=0,W=0.9914),Translation=(X=0,Y=-1 ...(truncated)"
    WeightMinimum=0
    Weight=@AlphaInterp.Result
   )
   n6 = &RigUnit_GetTransform(Item=(Type=Bone,Name="def_twist_corr_l"), Space=LocalSpace, bInitial=False)
   > &RigUnit_ModifyTransforms(
    ItemToModify="((Item=(Type=Bone,Name=\"MSTR_skirt_fr_l\"),Transform=(Rotation=(X=0.2619,Y=0.008,Z=-0.1372,W=0.9553),Translation=(X=0,Y=0,Z=0),Scale3D=(X=1,Y=1,Z=1))),(Item=(Type=Bone,Name=\"MSTR_skirt_fr_rear_l\"),Transform=(Rotation=(X=0.1337,Y=0.0317,Z=-0.092,W=0.9862),Tr ...(truncated)"
    WeightMinimum=0
    Weight=&RigVMFunction_MathDoubleRemap(
     SourceMaximum=-60
     SourceMinimum=-20
     TargetMaximum=1
     TargetMinimum=0
     Value=&RigVMFunction_MathQuaternionToEuler(RotationOrder=ZYX, Value=@n6.Transform.Rotation).Result.X
    ).Result
   )
   n7 = &RigUnit_GetTransform(Item=(Type=Bone,Name="calf_l"), Space=LocalSpace, bInitial=False)
   > &RigUnit_ModifyTransforms(
    ItemToModify="((Item=(Type=Bone,Name=\"MSTR_skirt_bk_l\"),Transform=(Rotation=(X=-0.1294,Y=0.017,Z=-0.1294,W=0.983),Translation=(X=0,Y=0,Z=0),Scale3D=(X=1,Y=1,Z=1))),(Item=(Type=Bone,Name=\"MSTR_skirt_side_rear_l\"),Transform=(Rotation=(X=-0.0151,Y=-0.0858,Z=-0.173,W=0.9811 ...(truncated)"
    WeightMinimum=0
    Weight=&RigVMFunction_MathDoubleRemap(
     SourceMaximum=60
     SourceMinimum=0
     TargetMaximum=1
     TargetMinimum=0
     Value=&RigVMFunction_MathQuaternionToEuler(RotationOrder=ZYX, Value=@n7.Transform.Rotation).Result.Z
    ).Result
   )
   n8 = &RigUnit_GetTransform(Item=(Type=Bone,Name="def_twist_corr_r"), Space=LocalSpace, bInitial=False)
   > &RigUnit_ModifyTransforms(
    ItemToModify="((Item=(Type=Bone,Name=\"MSTR_skirt_fr_r\"),Transform=(Rotation=(X=-0.2619,Y=-0.008,Z=-0.1372,W=0.9553),Translation=(X=0,Y=0,Z=0),Scale3D=(X=1,Y=1,Z=1))),(Item=(Type=Bone,Name=\"MSTR_skirt_fr_rear_r\"),Transform=(Rotation=(X=-0.1337,Y=-0.0317,Z=-0.092,W=0.9862 ...(truncated)"
    WeightMinimum=0
    Weight=&RigVMFunction_MathDoubleRemap(
     SourceMaximum=-60
     SourceMinimum=-20
     TargetMaximum=1
     TargetMinimum=0
     Value=&RigVMFunction_MathQuaternionToEuler(RotationOrder=ZYX, Value=@n8.Transform.Rotation).Result.X
    ).Result
   )
   n9 = &RigUnit_GetTransform(Item=(Type=Bone,Name="calf_r"), Space=LocalSpace, bInitial=False)
   > &RigUnit_ModifyTransforms(
    ItemToModify="((Item=(Type=Bone,Name=\"MSTR_skirt_bk_r\"),Transform=(Rotation=(X=0.1294,Y=-0.017,Z=-0.1294,W=0.983),Translation=(X=0,Y=0,Z=0),Scale3D=(X=1,Y=1,Z=1))),(Item=(Type=Bone,Name=\"MSTR_skirt_side_rear_r\"),Transform=(Rotation=(X=0.0151,Y=0.0858,Z=-0.173,W=0.9811), ...(truncated)"
    WeightMinimum=0
    Weight=&RigVMFunction_MathDoubleRemap(
     SourceMaximum=60
     SourceMinimum=0
     TargetMaximum=1
     TargetMinimum=0
     Value=&RigVMFunction_MathQuaternionToEuler(RotationOrder=ZYX, Value=@n9.Transform.Rotation).Result.Z
    ).Result
   )
   Get_Animation_Attribute = RigVMDispatchNode(BoneName=root, Default=0, Name=deformRig_l_thigh_up_alpha) {template="DISPATCH_RigDispatch_GetAnimAttribute(in Name,in BoneName,in Default,out Value,out Found)"}
   > &RigUnit_ModifyTransforms(
    ItemToModify="((Item=(Type=Bone,Name=\"MSTR_skirt_fr_l\"),Transform=(Rotation=(X=0.0437,Y=0.4332,Z=0.1017,W=0.8945),Translation=(X=3,Y=-5,Z=3),Scale3D=(X=1,Y=1,Z=1))),(Item=(Type=Bone,Name=\"MSTR_skirt_fr_rear_l\"),Transform=(Rotation=(X=0.0819,Y=0.3407,Z=-0.0298,W=0.9361), ...(truncated)"
    WeightMinimum=0
    Weight=@Get_Animation_Attribute.Value
   )
   Get_Animation_Attribute_1 = RigVMDispatchNode(BoneName=root, Default=0, Name=deformRig_r_thigh_up_alpha) {template="DISPATCH_RigDispatch_GetAnimAttribute(in Name,in BoneName,in Default,out Value,out Found)"}
   > &RigUnit_ModifyTransforms(
    ItemToModify="((Item=(Type=Bone,Name=\"MSTR_skirt_fr_r\"),Transform=(Rotation=(X=-0.0827,Y=-0.4372,Z=0.0827,W=0.8917),Translation=(X=5,Y=-3,Z=-3),Scale3D=(X=1,Y=1,Z=1))),(Item=(Type=Bone,Name=\"MSTR_skirt_fr_rear_r\"),Transform=(Rotation=(X=-0.0819,Y=-0.3407,Z=-0.0298,W=0.9 ...(truncated)"
    WeightMinimum=0
    Weight=@Get_Animation_Attribute_1.Value
   )
   > n10 = &RigUnit_SphericalPoseReader(
    ActiveRegionScaleFactors=(PositiveWidth=1.2,NegativeWidth=1,PositiveHeight=0.7,NegativeHeight=0.7)
    ActiveRegionSize=0.1
    Debug=(bDrawDebug=True,bDraw2D=False,bDrawLocalAxes=False,DebugScale=25,DebugSegments=20,DebugThickness=0.25)
    DriverAxis=(X=1,Y=0,Z=0)
    DriverItem=(Type=Bone,Name="thigh_r")
    FalloffRegionScaleFactors=(PositiveWidth=1.2,NegativeWidth=1,PositiveHeight=0.7,NegativeHeight=0.7)
    FalloffSize=0.2
    FlipHeightScaling=False
    FlipWidthScaling=False
    OptionalParentItem=(Type=Bone,Name="None")
    RotationOffset=(X=-90,Y=0,Z=0)
   )
   RigVMFunction_MathBoolOr = &RigVMFunction_MathBoolOr(B=RigVMVariableNode() {variable=isBaseballSliding_CR}.Value, A=RigVMVariableNode() {variable=isKneeSliding_CR}.Value)
   RigVMFunction_MathBoolNot = &RigVMFunction_MathBoolNot(Value=@RigVMFunction_MathBoolOr.Result)
   RigVMFunction_MathBoolAnd = &RigVMFunction_MathBoolAnd(B=@RigVMFunction_MathBoolNot.Result, A=RigVMVariableNode() {variable=isCrouching_CR}.Value)
   RigVMFunction_MathBoolToFloat = &RigVMFunction_MathBoolToFloat(Value=@RigVMFunction_MathBoolAnd.Result)
   n11 = &RigVMFunction_AlphaInterp(InterpSpeedDecreasing=5, InterpSpeedIncreasing=5, Scale=1, bClampResult=False, Value=@RigVMFunction_MathBoolToFloat.Result)
   n12 = &RigUnit_GetRelativeTransformForItem(Child=(Type=Bone,Name="pelvis"), Parent=(Type=Bone,Name="root"), bChildInitial=False, bParentInitial=False)
   > &RigUnit_ModifyTransforms(
    ItemToModify="((Item=(Type=Bone,Name=\"MSTR_skirt_fr_r\"),Transform=(Rotation=(X=-0.1648,Y=-0.1006,Z=0.0704,W=0.9786),Translation=(X=-2,Y=-5,Z=0),Scale3D=(X=1,Y=1,Z=1))),(Item=(Type=Bone,Name=\"MSTR_skirt_fr_rear_r\"),Transform=(Rotation=(X=-0.171,Y=-0.171,Z=-0.0302,W=0.969 ...(truncated)"
    WeightMinimum=0
    Weight=@n11.Result
    ItemToModify.0.Transform.Translation.Y=&RigVMFunction_MathDoubleRemap(SourceMaximum=65, SourceMinimum=40, TargetMaximum=-5, TargetMinimum=-2, Value=@n12.RelativeTransform.Translation.Z).Result
   )
   n13 = &RigVMFunction_MathQuaternionToEuler(
    RotationOrder=ZYX
    Value=&RigUnit_GetTransform(Item=(Type=Bone,Name="calf_l"), Space=LocalSpace, bInitial=False).Transform.Rotation
   )
   > &RigUnit_ModifyTransforms(
    ItemToModify=((Item=(Type=Bone,Name="def_twist_corr_bk_l"),Transform=(Rotation=(X=-0.2566,Y=-0.0338,Z=0.1261,W=0.9577),Translation=(X=0,Y=-1,Z=3),Scale3D=(X=1,Y=1,Z=1.5))))
    WeightMinimum=0
    Weight=&RigVMFunction_MathDoubleMul(
     B=&RigVMFunction_MathDoubleRemap(SourceMaximum=40, SourceMinimum=65, TargetMaximum=1, TargetMinimum=0, Value=@n12.RelativeTransform.Translation.Z).Result
     A=&RigVMFunction_MathDoubleRemap(SourceMaximum=115, SourceMinimum=90, TargetMaximum=1, TargetMinimum=0, Value=@n13.Result.Z).Result
    ).Result
   )
   RigVMFunction_MathQuaternionToEuler = &RigVMFunction_MathQuaternionToEuler(
    RotationOrder=ZYX
    Value=&RigUnit_GetTransform(Item=(Type=Bone,Name="thigh_l"), Space=LocalSpace, bInitial=False).Transform.Rotation
   )
   > &RigUnit_ModifyTransforms(
    ItemToModify="((Item=(Type=Bone,Name=\"MSTR_skirt_side_l\"),Transform=(Rotation=(X=0,Y=0,Z=-0.7071,W=0.7071),Translation=(X=0,Y=0,Z=0),Scale3D=(X=1,Y=1,Z=1))),(Item=(Type=Bone,Name=\"MSTR_skirt_side_rear_l\"),Transform=(Rotation=(X=0,Y=0,Z=-0.2588,W=0.9659),Translation=(X=0 ...(truncated)"
    WeightMinimum=0
    Weight=&RigVMFunction_MathDoubleRemap(SourceMaximum=-80, SourceMinimum=-15, TargetMaximum=1, TargetMinimum=0, Value=@RigVMFunction_MathQuaternionToEuler.Result.Y).Result
   )
   n14 = &RigUnit_GetTransform(Item=(Type=Bone,Name="thigh_r"), Space=LocalSpace, bInitial=False)
   > &RigUnit_ModifyTransforms(
    ItemToModify="((Item=(Type=Bone,Name=\"MSTR_skirt_side_r\"),Transform=(Rotation=(X=0,Y=0,Z=-0.7071,W=0.7071),Translation=(X=0,Y=0,Z=0),Scale3D=(X=1,Y=1,Z=1))),(Item=(Type=Bone,Name=\"MSTR_skirt_side_rear_r\"),Transform=(Rotation=(X=0,Y=0,Z=-0.2588,W=0.9659),Translation=(X=0 ...(truncated)"
    WeightMinimum=0
    Weight=&RigVMFunction_MathDoubleRemap(
     SourceMaximum=80
     SourceMinimum=15
     TargetMaximum=1
     TargetMinimum=0
     Value=&RigVMFunction_MathQuaternionToEuler(RotationOrder=ZYX, Value=@n14.Transform.Rotation).Result.Y
    ).Result
   )
   n15 = &RigVMFunction_MathQuaternionToEuler(
    RotationOrder=ZYX
    Value=&RigUnit_GetTransform(Item=(Type=Bone,Name="calf_l"), Space=LocalSpace, bInitial=False).Transform.Rotation
   )
   Multiply = &RigVMFunction_MathDoubleMul(
    B=&RigVMFunction_MathDoubleRemap(SourceMaximum=90, SourceMinimum=20, TargetMaximum=1, TargetMinimum=0, Value=@n15.Result.Z).Result
    A=&RigVMFunction_MathDoubleRemap(
     SourceMaximum=1
     SourceMinimum=0.2
     TargetMaximum=0
     TargetMinimum=1
     Value=RigVMDispatchNode(BoneName=root, Default=1, Name=deformRig_l_thigh_up_alpha) {template="DISPATCH_RigDispatch_GetAnimAttribute(in Name,in BoneName,in Default,out Value,out Found)"}.Value
    ).Result
   )
   > &RigUnit_ModifyTransforms(
    ItemToModify="((Item=(Type=Bone,Name=\"MSTR_skirt_bk_l\"),Transform=(Rotation=(X=-0.1337,Y=-0.0317,Z=-0.092,W=0.9862),Translation=(X=-2,Y=-2,Z=0),Scale3D=(X=1,Y=1,Z=1))),(Item=(Type=Bone,Name=\"MSTR_skirt_side_rear_l\"),Transform=(Rotation=(X=-0.0896,Y=-0.0381,Z=-0.0646,W=0 ...(truncated)"
    WeightMinimum=0
    Weight=@Multiply.Result
   )
   n16 = &RigVMFunction_MathBoolOr(
    A=&RigVMFunction_MathBoolNot(Value=RigVMVariableNode() {variable=Rban}.Value).Result
    B=RigVMVariableNode() {variable=isDBNO}.Value
   )
   > &RigUnit_ModifyTransforms(
    ItemToModify=((Item=(Type=Bone,Name="def_alt_skirt_corr_l"),Transform=(Rotation=(X=0.0436,Y=-0.0755,Z=-0.8627,W=0.4981),Translation=(X=-12,Y=5,Z=5),Scale3D=(X=0.9,Y=1,Z=0.7))))
    WeightMinimum=0
    Weight=&RigVMFunction_MathDoubleMul(
     A=&RigVMFunction_MathBoolToFloat(Value=@n16.Result).Result
     B=&RigVMFunction_MathDoubleRemap(SourceMaximum=140, SourceMinimum=20, TargetMaximum=1, TargetMinimum=0, Value=@n15.Result.Z).Result
    ).Result
   )
   n17 = &RigVMFunction_MathQuaternionToEuler(
    RotationOrder=ZYX
    Value=&RigUnit_GetTransform(Item=(Type=Bone,Name="calf_r"), Space=LocalSpace, bInitial=False).Transform.Rotation
   )
   n18 = RigVMDispatchNode(BoneName=root, Default=1, Name=deformRig_r_thigh_up_alpha) {template="DISPATCH_RigDispatch_GetAnimAttribute(in Name,in BoneName,in Default,out Value,out Found)"}
   > &RigUnit_ModifyTransforms(
    ItemToModify="((Item=(Type=Bone,Name=\"MSTR_skirt_bk_r\"),Transform=(Rotation=(X=0.0038,Y=0.0435,Z=-0.0871,W=0.9952),Translation=(X=-2,Y=-2,Z=0),Scale3D=(X=1,Y=1,Z=1))),(Item=(Type=Bone,Name=\"MSTR_skirt_side_rear_r\"),Transform=(Rotation=(X=0.0896,Y=0.0381,Z=-0.0646,W=0.99 ...(truncated)"
    WeightMinimum=0
    Weight=&RigVMFunction_MathDoubleMul(
     B=&RigVMFunction_MathDoubleRemap(SourceMaximum=90, SourceMinimum=20, TargetMaximum=1, TargetMinimum=0, Value=@n17.Result.Z).Result
     A=&RigVMFunction_MathDoubleRemap(SourceMaximum=1, SourceMinimum=0.2, TargetMaximum=0, TargetMinimum=1, Value=@n18.Value).Result
    ).Result
   )
   n19 = &RigVMFunction_MathBoolOr(
    A=&RigVMFunction_MathBoolNot(Value=RigVMVariableNode() {variable=Rban}.Value).Result
    B=RigVMVariableNode() {variable=isDBNO}.Value
   )
   > &RigUnit_ModifyTransforms(
    ItemToModify=((Item=(Type=Bone,Name="def_alt_skirt_corr_r"),Transform=(Rotation=(X=-0.0436,Y=0.0755,Z=-0.8627,W=0.4981),Translation=(X=-12,Y=5,Z=-5),Scale3D=(X=0.9,Y=1,Z=0.7))))
    WeightMinimum=0
    Weight=&RigVMFunction_MathDoubleMul(
     A=&RigVMFunction_MathBoolToFloat(Value=@n19.Result).Result
     B=&RigVMFunction_MathDoubleRemap(SourceMaximum=140, SourceMinimum=20, TargetMaximum=1, TargetMinimum=0, Value=@n17.Result.Z).Result
    ).Result
   )
   > n20 = &RigUnit_SphericalPoseReader(
    ActiveRegionScaleFactors=(PositiveWidth=0.5,NegativeWidth=2,PositiveHeight=0.7,NegativeHeight=0.7)
    ActiveRegionSize=0.1
    Debug=(bDrawDebug=false,bDraw2D=False,bDrawLocalAxes=False,DebugScale=25,DebugSegments=20,DebugThickness=0.25)
    DriverAxis=(X=1,Y=0,Z=0)
    DriverItem=(Type=Bone,Name="thigh_r")
    FalloffRegionScaleFactors=(PositiveWidth=0.5,NegativeWidth=1,PositiveHeight=0.7,NegativeHeight=0.7)
    FalloffSize=0.2
    FlipHeightScaling=False
    FlipWidthScaling=False
    OptionalParentItem=(Type=Bone,Name="None")
    RotationOffset=(X=90,Y=0,Z=-60)
   )
   > &RigUnit_ModifyTransforms(
    ItemToModify="((Item=(Type=Bone,Name=\"MSTR_skirt_bk_r\"),Transform=(Rotation=(X=0,Y=0,Z=-0.2588,W=0.9659),Translation=(X=0,Y=0,Z=0),Scale3D=(X=1,Y=1,Z=1))),(Item=(Type=Bone,Name=\"MSTR_skirt_side_rear_r\"),Transform=(Rotation=(X=-0.017,Y=-0.1294,Z=-0.1294,W=0.983),Translat ...(truncated)"
    WeightMinimum=0
    Weight=@n20.OutputParam
   )
   > n21 = &RigUnit_SphericalPoseReader(
    ActiveRegionScaleFactors=(PositiveWidth=0.5,NegativeWidth=2,PositiveHeight=0.7,NegativeHeight=0.7)
    ActiveRegionSize=0.1
    Debug=(bDrawDebug=false,bDraw2D=False,bDrawLocalAxes=False,DebugScale=25,DebugSegments=20,DebugThickness=0.25)
    DriverAxis=(X=-1,Y=0,Z=0)
    DriverItem=(Type=Bone,Name="thigh_l")
    FalloffRegionScaleFactors=(PositiveWidth=0.5,NegativeWidth=1,PositiveHeight=0.7,NegativeHeight=0.7)
    FalloffSize=0.2
    FlipHeightScaling=false
    FlipWidthScaling=true
    OptionalParentItem=(Type=Bone,Name="None")
    RotationOffset=(X=-90,Y=0,Z=-60)
   )
   > &RigUnit_ModifyTransforms(
    ItemToModify="((Item=(Type=Bone,Name=\"MSTR_skirt_bk_l\"),Transform=(Rotation=(X=0,Y=0,Z=-0.2588,W=0.9659),Translation=(X=0,Y=0,Z=0),Scale3D=(X=1,Y=1,Z=1))),(Item=(Type=Bone,Name=\"MSTR_skirt_side_rear_l\"),Transform=(Rotation=(X=0.017,Y=0.1294,Z=-0.1294,W=0.983),Translatio ...(truncated)"
    WeightMinimum=0
    Weight=@n21.OutputParam
   )
   n22 = &RigVMFunction_MathBoolToFloat(Value=RigVMVariableNode() {variable=isKneeSliding_CR}.Value)
   > &RigUnit_ModifyTransforms(
    ItemToModify="((Item=(Type=Bone,Name=\"MSTR_skirt_fr_l\"),Transform=(Rotation=(X=0.2831,Y=0.2969,Z=0.0704,W=0.9093),Translation=(X=-1,Y=-2,Z=-5),Scale3D=(X=1,Y=1,Z=1))),(Item=(Type=Bone,Name=\"MSTR_skirt_fr_rear_l\"),Transform=(Rotation=(X=0.2431,Y=0.3153,Z=-0.3602,W=0.8437 ...(truncated)"
    WeightMinimum=0
    Weight=@n22.Result
   )
   > &RigUnit_ModifyTransforms(
    ItemToModify="((Item=(Type=Bone,Name=\"MSTR_skirt_fr_r\"),Transform=(Rotation=(X=-0.2831,Y=-0.2969,Z=0.0704,W=0.9093),Translation=(X=-1,Y=-2,Z=5),Scale3D=(X=1,Y=1,Z=1))),(Item=(Type=Bone,Name=\"MSTR_skirt_fr_rear_r\"),Transform=(Rotation=(X=-0.1532,Y=-0.3675,Z=-0.1296,W=0.9 ...(truncated)"
    WeightMinimum=0
    Weight=@n22.Result
   )
   n23 = &RigVMFunction_MathBoolToFloat(Value=RigVMVariableNode() {variable=isBaseballSliding_CR}.Value)
   n24 = &RigVMFunction_AlphaInterp(InterpSpeedDecreasing=10, InterpSpeedIncreasing=10, Scale=1, bClampResult=False, Value=@n23.Result)
   > &RigUnit_ModifyTransforms(
    ItemToModify="((Item=(Type=Bone,Name=\"MSTR_skirt_fr_l\"),Transform=(Rotation=(X=0.1261,Y=0.092,Z=0.0317,W=0.9872),Translation=(X=0,Y=-1,Z=0),Scale3D=(X=1,Y=1,Z=1))),(Item=(Type=Bone,Name=\"MSTR_skirt_fr_rear_l\"),Transform=(Rotation=(X=0.0868,Y=-0.0868,Z=0.0076,W=0.9924),T ...(truncated)"
    WeightMinimum=0
    Weight=@n24.Result
   )
   > &RigUnit_ModifyTransforms(
    ItemToModify="((Item=(Type=Bone,Name=\"MSTR_skirt_fr_r\"),Transform=(Rotation=(X=0.4101,Y=-0.0097,Z=-0.7943,W=0.4481),Translation=(X=5,Y=0,Z=-3),Scale3D=(X=1,Y=1,Z=1))),(Item=(Type=Bone,Name=\"MSTR_skirt_fr_rear_r\"),Transform=(Rotation=(X=0.2336,Y=-0.0626,Z=-0.855,W=0.4588 ...(truncated)"
    WeightMinimum=0
    Weight=@n24.Result
   )
   n25 = &RigVMFunction_MathQuaternionToEuler(
    RotationOrder=XYZ
    Value=&RigUnit_GetRelativeTransformForItem(Child=(Type=Bone,Name="spine_05"), Parent=(Type=Bone,Name="pelvis"), bChildInitial=False, bParentInitial=False).RelativeTransform.Rotation
   )
   n26 = &RigVMFunction_AlphaInterp(
    InterpSpeedDecreasing=5
    InterpSpeedIncreasing=5
    Scale=1
    bClampResult=False
    Value=&RigVMFunction_MathDoubleMul(
     B=&RigVMFunction_MathDoubleRemap(SourceMaximum=45, SourceMinimum=0, TargetMaximum=1, TargetMinimum=0, Value=@n25.Result.X).Result
     A=@n23.Result
    ).Result
   )
   > &RigUnit_ModifyTransforms(
    ItemToModify="((Item=(Type=Bone,Name=\"MSTR_skirt_fr_l\"),Transform=(Rotation=(X=0.0905,Y=-0.0357,Z=0.0905,W=0.9911),Translation=(X=0,Y=0,Z=0),Scale3D=(X=1,Y=1,Z=1))),(Item=(Type=Bone,Name=\"MSTR_skirt_fr_rear_l\"),Transform=(Rotation=(X=0,Y=0.0872,Z=0,W=0.9962),Translation ...(truncated)"
    WeightMinimum=0
    Weight=@n26.Result
   )
   > &RigUnit_ModifyTransforms(
    ItemToModify="((Item=(Type=Bone,Name=\"MSTR_skirt_fr_r\"),Transform=(Rotation=(X=0.1063,Y=-0.2495,Z=0.1063,W=0.9566),Translation=(X=0,Y=-5,Z=5),Scale3D=(X=1,Y=1,Z=1))),(Item=(Type=Bone,Name=\"MSTR_skirt_fr_rear_r\"),Transform=(Rotation=(X=0.096,Y=-0.2896,Z=-0.2028,W=0.9305) ...(truncated)"
    WeightMinimum=0
    Weight=@n26.Result
   )
   n27 = &RigVMFunction_AlphaInterp(
    InterpSpeedDecreasing=5
    InterpSpeedIncreasing=5
    Scale=1
    bClampResult=False
    Value=&RigVMFunction_MathDoubleMul(
     B=&RigVMFunction_MathDoubleRemap(SourceMaximum=-50, SourceMinimum=-10, TargetMaximum=1, TargetMinimum=0, Value=@n25.Result.X).Result
     A=@n23.Result
    ).Result
   )
   > &RigUnit_ModifyTransforms(
    ItemToModify="((Item=(Type=Bone,Name=\"MSTR_skirt_fr_l\"),Transform=(Rotation=(X=0.1555,Y=0.1146,Z=0.1555,W=0.9688),Translation=(X=0,Y=0,Z=0),Scale3D=(X=1,Y=1,Z=1))),(Item=(Type=Bone,Name=\"MSTR_skirt_fr_rear_l\"),Transform=(Rotation=(X=-0.171,Y=0.171,Z=0.0302,W=0.9698),Tra ...(truncated)"
    WeightMinimum=0
    Weight=@n27.Result
   )
   > &RigUnit_ModifyTransforms(
    ItemToModify="((Item=(Type=Bone,Name=\"MSTR_skirt_fr_r\"),Transform=(Rotation=(X=-0.2195,Y=0.3135,Z=0.5299,W=0.7568),Translation=(X=0,Y=-5,Z=-3),Scale3D=(X=1,Y=1,Z=1))),(Item=(Type=Bone,Name=\"MSTR_skirt_fr_rear_r\"),Transform=(Rotation=(X=0.0915,Y=0.3415,Z=0.4085,W=0.8415) ...(truncated)"
    WeightMinimum=0
    Weight=@n27.Result
   )
   n28 = &RigVMFunction_MathQuaternionToEuler(
    RotationOrder=ZYX
    Value=&RigUnit_GetTransform(Item=(Type=Bone,Name="foot_l"), Space=LocalSpace, bInitial=False).Transform.Rotation
   )
   Remap = &RigVMFunction_MathDoubleRemap(SourceMaximum=60, SourceMinimum=0, TargetMaximum=1, TargetMinimum=0, Value=@n28.Result.Z)
   &RigUnit_ModifyTransforms(
    ItemToModify="((Item=(Type=Bone,Name=\"foot_l\"),Transform=(Rotation=(X=0,Y=0,Z=0,W=1),Translation=(X=0,Y=0,Z=0),Scale3D=(X=1,Y=1,Z=1))),(Item=(Type=Bone,Name=\"def_foot_corr_bk_l\"),Transform=(Rotation=(X=0,Y=0,Z=-0.2588,W=0.9659),Translation=(X=3,Y=-6,Z=0.95),Scale3D=(X=1 ...(truncated)"
    WeightMinimum=0
    Weight=@Remap.Result
    ExecutePin=@RigVMFunction_Sequence.D
   )
   > &RigUnit_ModifyTransforms(
    ItemToModify="((Item=(Type=Bone,Name=\"foot_l\"),Transform=(Rotation=(X=0,Y=0,Z=0,W=1),Translation=(X=0,Y=0,Z=0),Scale3D=(X=1,Y=1,Z=1))),(Item=(Type=Bone,Name=\"def_foot_corr_fr_l\"),Transform=(Rotation=(X=0,Y=0,Z=0.4267,W=-0.9044),Translation=(X=0.7187,Y=-5.4261,Z=0),Scale ...(truncated)"
    WeightMinimum=0
    Weight=&RigVMFunction_MathDoubleRemap(SourceMaximum=-30, SourceMinimum=0, TargetMaximum=1, TargetMinimum=0, Value=@n28.Result.Z).Result
   )
   n29 = &RigVMFunction_MathQuaternionToEuler(
    RotationOrder=ZYX
    Value=&RigUnit_GetTransform(Item=(Type=Bone,Name="foot_r"), Space=LocalSpace, bInitial=False).Transform.Rotation
   )
   > &RigUnit_ModifyTransforms(
    ItemToModify=((Item=(Type=Bone,Name="def_foot_corr_bk_r"),Transform=(Rotation=(X=0,Y=0,Z=-0.2588,W=0.9659),Translation=(X=3,Y=-6,Z=-0.95),Scale3D=(X=1,Y=1,Z=1))))
    WeightMinimum=0
    Weight=&RigVMFunction_MathDoubleRemap(SourceMaximum=60, SourceMinimum=0, TargetMaximum=1, TargetMinimum=0, Value=@n29.Result.Z).Result
   )
   > &RigUnit_ModifyTransforms(
    ItemToModify=((Item=(Type=Bone,Name="def_foot_corr_fr_r"),Transform=(Rotation=(X=0,Y=0,Z=0.4267,W=-0.9044),Translation=(X=0.7187,Y=-5.4261,Z=0),Scale3D=(X=1,Y=1,Z=1))))
    WeightMinimum=0
    Weight=&RigVMFunction_MathDoubleRemap(SourceMaximum=-30, SourceMinimum=0, TargetMaximum=1, TargetMinimum=0, Value=@n29.Result.Z).Result
   )
   > &RigUnit_ParentConstraint(
    AdvancedSettings=(InterpolationType=Average,RotationOrderForFilter=XZY)
    Child=(Type=Bone,Name="base_shoulderPad_l")
    Parents="((Item=(Type=Bone,Name=\"upperarm_l\"),Weight=0.8),(Item=(Type=Bone,Name=\"upperarm_twist_01_l\"),Weight=0.2),(Item=(Type=Bone,Name=\"clavicle_l\"),Weight=0),(Item=(Type=Bone,Name=\"deform_upperarm_up_l\"),Weight=0.2),(Item=(Type=Bone,Name=\"deform_upperarm_fr_l\"), ...(truncated)"
    Weight=1
   )
   > &RigUnit_ParentConstraint(
    AdvancedSettings=(InterpolationType=Average,RotationOrderForFilter=XZY)
    Child=(Type=Bone,Name="base_shoulderPad_r")
    Parents="((Item=(Type=Bone,Name=\"upperarm_r\"),Weight=0.8),(Item=(Type=Bone,Name=\"upperarm_twist_01_r\"),Weight=0.2),(Item=(Type=Bone,Name=\"clavicle_r\"),Weight=0),(Item=(Type=Bone,Name=\"deform_upperarm_up_r\"),Weight=0.2),(Item=(Type=Bone,Name=\"deform_upperarm_fr_r\"), ...(truncated)"
    Weight=1
   )
   > &RigUnit_ModifyTransforms(
    ItemToModify=((Item=(Type=Bone,Name="def_shoulderPad_l"),Transform=(Rotation=(X=0,Y=0,Z=0,W=1),Translation=(X=1,Y=0,Z=0),Scale3D=(X=0.95,Y=0.95,Z=0.95))))
    Weight=1
    WeightMinimum=0
   )
   > &RigUnit_ModifyTransforms(
    ItemToModify=((Item=(Type=Bone,Name="def_shoulderPad_r"),Transform=(Rotation=(X=0,Y=0,Z=0,W=1),Translation=(X=1,Y=0,Z=0),Scale3D=(X=0.95,Y=0.95,Z=0.95))))
    Weight=1
    WeightMinimum=0
   )
   n30 = &RigVMFunction_MathQuaternionToEuler(
    RotationOrder=XZY
    Value=&RigUnit_GetTransform(Item=(Type=Bone,Name="upperarm_twist_01_l"), Space=LocalSpace, bInitial=False).Transform.Rotation
   )
   > &RigUnit_ModifyTransforms(
    ItemToModify=((Item=(Type=Bone,Name="base_shoulderPad_l"),Transform=(Rotation=(X=0,Y=0,Z=0,W=1),Translation=(X=0,Y=0,Z=0),Scale3D=(X=1,Y=1,Z=1))))
    Weight=1
    WeightMinimum=0
    ItemToModify.0.Transform.Rotation=&RigVMFunction_MathQuaternionFromEuler(
     Euler=(X=0,Y=0,Z=0)
     RotationOrder=XZY
     Euler.X=&RigVMFunction_MathDoubleRemap(SourceMaximum=-20, SourceMinimum=20, TargetMaximum=-10, TargetMinimum=10, Value=@n30.Result.X).Result
    ).Result
   )
   n31 = &RigVMFunction_MathQuaternionToEuler(
    RotationOrder=XZY
    Value=&RigUnit_GetTransform(Item=(Type=Bone,Name="upperarm_twist_01_r"), Space=LocalSpace, bInitial=False).Transform.Rotation
   )
   > &RigUnit_ModifyTransforms(
    ItemToModify=((Item=(Type=Bone,Name="base_shoulderPad_r"),Transform=(Rotation=(X=0,Y=0,Z=0,W=1),Translation=(X=0,Y=0,Z=0),Scale3D=(X=1,Y=1,Z=1))))
    Weight=1
    WeightMinimum=0
    ItemToModify.0.Transform.Rotation=&RigVMFunction_MathQuaternionFromEuler(
     Euler=(X=0,Y=0,Z=0)
     RotationOrder=XZY
     Euler.X=&RigVMFunction_MathDoubleRemap(SourceMaximum=-20, SourceMinimum=20, TargetMaximum=-10, TargetMinimum=10, Value=@n31.Result.X).Result
    ).Result
   )
   Get_Animation_Attribute_2 = RigVMDispatchNode(BoneName=root, Default=0, Name=deformRig_l_upperarm_up_alpha) {template="DISPATCH_RigDispatch_GetAnimAttribute(in Name,in BoneName,in Default,out Value,out Found)"}
   Get_Animation_Attribute_6 = RigVMDispatchNode(BoneName=root, Default=0, Name=deformRig_l_upperarm_fr_alpha) {template="DISPATCH_RigDispatch_GetAnimAttribute(in Name,in BoneName,in Default,out Value,out Found)"}
   > &RigUnit_ModifyTransforms(
    ItemToModify=((Item=(Type=Bone,Name="def_shoulderPad_l"),Transform=(Rotation=(X=0,Y=0,Z=0.1305,W=0.9914),Translation=(X=2.5,Y=-1,Z=0),Scale3D=(X=1,Y=1,Z=1))))
    WeightMinimum=0
    Weight=&RigVMFunction_MathDoubleMul(
     A=@Get_Animation_Attribute_2.Value
     B=&RigVMFunction_MathDoubleRemap(SourceMaximum=1, SourceMinimum=0, TargetMaximum=0, TargetMinimum=1, Value=@Get_Animation_Attribute_6.Value).Result
    ).Result
   )
   Get_Animation_Attribute_3 = RigVMDispatchNode(BoneName=root, Default=0, Name=deformRig_r_upperarm_up_alpha) {template="DISPATCH_RigDispatch_GetAnimAttribute(in Name,in BoneName,in Default,out Value,out Found)"}
   Get_Animation_Attribute_7 = RigVMDispatchNode(BoneName=root, Default=0, Name=deformRig_r_upperarm_fr_alpha) {template="DISPATCH_RigDispatch_GetAnimAttribute(in Name,in BoneName,in Default,out Value,out Found)"}
   > &RigUnit_ModifyTransforms(
    ItemToModify=((Item=(Type=Bone,Name="def_shoulderPad_r"),Transform=(Rotation=(X=0,Y=0,Z=0.1305,W=0.9914),Translation=(X=2,Y=-1,Z=0),Scale3D=(X=1,Y=1,Z=1))))
    WeightMinimum=0
    Weight=&RigVMFunction_MathDoubleMul(
     A=@Get_Animation_Attribute_3.Value
     B=&RigVMFunction_MathDoubleRemap(SourceMaximum=1, SourceMinimum=0, TargetMaximum=0, TargetMinimum=1, Value=@Get_Animation_Attribute_7.Value).Result
    ).Result
   )
   Cast = &RigVMFunction_MathBoolToFloat(Value=&RigVMFunction_MathBoolNot(Value=RigVMVariableNode() {variable=Rban}.Value).Result)
   &RigUnit_ModifyTransforms(
    ExecutePin=()
    ItemToModify="((Item=(Type=Bone,Name=\"dyn_skirt_fr_02_l\"),Transform=(Rotation=(X=-0.0151,Y=0.173,Z=0.0858,W=0.9811),Translation=(X=0,Y=0,Z=0),Scale3D=(X=1,Y=1,Z=1))),(Item=(Type=Bone,Name=\"dyn_skirt_fr_rear_02_l\"),Transform=(Rotation=(X=0.1555,Y=0.1853,Z=0.0545,W=0.9688 ...(truncated)"
    WeightMinimum=0
    Weight=&RigVMFunction_MathDoubleMul(B=@Cast.Result, A=@Get_Animation_Attribute.Value).Result
   )
   &RigUnit_ModifyTransforms(
    ExecutePin=()
    ItemToModify=((Item=(Type=Bone,Name="def_alt_skirt_corr_l"),Transform=(Rotation=(X=-0.0367,Y=-0.1517,Z=-0.9242,W=0.3485),Translation=(X=-12.2,Y=6,Z=0.8),Scale3D=(X=0.66,Y=1.008,Z=1))))
    Weight=1
    WeightMinimum=0
   )
   &RigUnit_ModifyTransforms(
    ExecutePin=()
    ItemToModify=((Item=(Type=Bone,Name="def_alt_skirt_corr_r"),Transform=(Rotation=(X=0.0367,Y=0.1517,Z=-0.9242,W=0.3485),Translation=(X=-12.2,Y=6,Z=-0.8),Scale3D=(X=0.66,Y=1.008,Z=1))))
    Weight=1
    WeightMinimum=0
   )
   n32 = &RigVMFunction_MathBoolNot(Value=RigVMVariableNode() {variable=Rban}.Value)
   &RigUnit_ModifyTransforms(
    ExecutePin=()
    ItemToModify="((Item=(Type=Bone,Name=\"dyn_skirt_fr_02_r\"),Transform=(Rotation=(X=0.0151,Y=-0.173,Z=0.0858,W=0.9811),Translation=(X=0,Y=0,Z=0),Scale3D=(X=1,Y=1,Z=1))),(Item=(Type=Bone,Name=\"dyn_skirt_fr_rear_02_r\"),Transform=(Rotation=(X=-0.1555,Y=-0.1853,Z=0.0545,W=0.96 ...(truncated)"
    WeightMinimum=0
    Weight=&RigVMFunction_MathDoubleMul(B=&RigVMFunction_MathBoolToFloat(Value=@n32.Result).Result, A=@Get_Animation_Attribute_1.Value).Result
   )
   n33 = &RigVMFunction_MathDoubleRemap(SourceMaximum=65, SourceMinimum=40, TargetMaximum=1, TargetMinimum=0, Value=@n12.RelativeTransform.Translation.Z)
   &RigUnit_ModifyTransforms(
    ExecutePin=()
    ItemToModify="((Item=(Type=Bone,Name=\"MSTR_skirt_fr_r\"),Transform=(Rotation=(X=0,Y=0,Z=0,W=1),Translation=(X=2,Y=0,Z=0),Scale3D=(X=1,Y=1,Z=1))),(Item=(Type=Bone,Name=\"MSTR_skirt_fr_rear_r\"),Transform=(Rotation=(X=0,Y=0,Z=0,W=1),Translation=(X=0,Y=0,Z=0),Scale3D=(X=1,Y=1 ...(truncated)"
    WeightMinimum=0
    Weight=&RigVMFunction_MathDoubleMul(B=&RigVMFunction_MathDoubleMul(B=@n11.Result, A=@n33.Result).Result, A=@n10.OutputParam).Result
   )
   &RigUnit_ModifyTransforms(
    ExecutePin=()
    ItemToModify=((Item=(Type=Bone,Name="calf_r"),Transform=(Rotation=(X=0,Y=0,Z=0.9397,W=0.342),Translation=(X=0,Y=0,Z=0),Scale3D=(X=1,Y=1,Z=1))))
    Weight=1
    WeightMinimum=0
   )
   &RigUnit_ModifyTransforms(
    ExecutePin=()
    ItemToModify=((Item=(Type=Bone,Name="thigh_r"),Transform=(Rotation=(X=0,Y=0,Z=0.2588,W=0.9659),Translation=(X=0,Y=0,Z=0),Scale3D=(X=1,Y=1,Z=1))),(Item=(Type=Bone,Name="calf_r"),Transform=(Rotation=(X=0,Y=0,Z=0,W=1),Translation=(X=0,Y=0,Z=0),Scale3D=(X=1,Y=1,Z=1))))
    Weight=1
    WeightMinimum=0
   )
   &RigUnit_ParentConstraint(
    AdvancedSettings=(InterpolationType=Average,RotationOrderForFilter=XYZ)
    Child=(Type=Bone,Name="def_alt_skirt_corr_l")
    ExecutePin=()
    Filter=(TranslationFilter=(bX=false,bY=false,bZ=false),RotationFilter=(bX=true,bY=true,bZ=true),ScaleFilter=(bX=True,bY=True,bZ=True))
    Parents=((Item=(Type=Bone,Name="thigh_l"),Weight=1),(Item=(Type=Bone,Name="def_twist_corr_bk_l"),Weight=1))
    Weight=1
   )
  ]}
]
hierarchy [
 root {loc=<0 0 0>, rot=<0 0 0 1>}
 pelvis {parent=root, loc=<0 -0.7071 93.1191>, rot=<0.0127 0.707 -0.0127 -0.707>}
 spine_01 {parent=pelvis, loc=<6.7494 -0.6388 0>, rot=<0 0 0.034 -0.9994>}
 spine_02 {parent=spine_01, loc=<7.2518 -0.035 0>, rot=<0 0 0.0238 -0.9997>}
 spine_03 {parent=spine_02, loc=<6.0104 0 0>, rot=<0 0 0.0149 -0.9999>}
 spine_04 {parent=spine_03, loc=<8.1024 0 0>, rot=<0 0 -0.0132 -0.9999>}
 spine_05 {parent=spine_04, loc=<10.0146 0 0>, rot=<0 0 -0.0015 -1>}
 neck_01 {parent=spine_05, loc=<11.3138 0.1529 0>, rot=<0 0 -0.2003 -0.9797>}
 neck_02 {parent=neck_01, loc=<4.5627 0 0>, rot=<0 0 0.0504 -0.9987>}
 head {parent=neck_02, loc=<5.3914 0 0>, rot=<0 0 0.115 -0.9934>}
 dyn_curl_bk {parent=head, loc=<-3.3699 -4.9444 0.1205>, rot=<0.4448 -0.8456 -0.0581 -0.2894>}
 dyn_curl_bk_side_03_l {parent=head, loc=<-0.8977 -2.1588 -5.825>, rot=<0.4338 -0.6461 -0.384 -0.497>}
 dyn_curl_bk_side_04_l {parent=head, loc=<-1.9043 -2.7811 -4.1633>, rot=<0.2656 -0.6116 -0.5421 -0.5115>}
 dyn_curl_bk_side_03_r {parent=head, loc=<-0.7433 -1.8523 4.7346>, rot=<-0.2789 0.5003 -0.5788 -0.5805>}
 attach_cape {parent=spine_05, loc=<9.017 -11.714 0>, rot=<0.0616 -0.7044 -0.0616 -0.7044>}
 cape {parent=attach_cape, type=Null, loc=<0 4.251 0>, rot=<0 0 0 1>}
 attach_backpack {parent=spine_05, loc=<-0.123 -13.473 0>, rot=<-0.0093 -0.707 0.0093 -0.707>}
 Backpack_BR {parent=attach_backpack, type=Null, loc=<0 3.0405 0>, rot=<0 0 0 1>}
 clavicle_l {parent=spine_05, loc=<4.3575 -0.1466 -2.3223>, rot=<0.0465 -0.7289 -0.009 -0.6829>}
 upperarm_l {parent=clavicle_l, loc=<15.5111 0 0>, rot=<0.0226 -0.3234 -0.0088 -0.9459>}
 lowerarm_l {parent=upperarm_l, loc=<26.6514 0 0>, rot=<0.0391 -0.0099 -0.2453 -0.9686>}
 hand_l {parent=lowerarm_l, loc=<22.3491 0 0>, rot=<-0.7081 -0.0507 0.0975 -0.6975>}
 thumb_01_l {parent=hand_l, loc=<1.3118 -0.3191 -2.459>, rot=<0.6546 -0.4042 0.0664 -0.6354>}
 thumb_02_l {parent=thumb_01_l, loc=<5.9336 -0.0448 0.0401>, rot=<-0.0037 0.0213 0.298 -0.9543>}
 thumb_03_l {parent=thumb_02_l, loc=<2.2223 0.3682 -0.0907>, rot=<0.0118 -0.013 0.0066 -0.9998>}
 index_metacarpal_l {parent=hand_l, loc=<1.6535 0.0183 -2.96>, rot=<0.0158 -0.0124 -0.0173 -0.9996>}
 index_01_l {parent=index_metacarpal_l, loc=<8.3008 0.0357 -0.0524>, rot=<-0.0044 0.0485 0.0924 -0.9945>}
 index_02_l {parent=index_01_l, loc=<3.3587 -0.3004 0.0452>, rot=<-0.002 0.0027 0.3225 -0.9466>}
 index_03_l {parent=index_02_l, loc=<2.5578 -0.123 -0.0252>, rot=<-0.002 0.0067 0.2558 -0.9667>}
 middle_metacarpal_l {parent=hand_l, loc=<2.1856 -0.1725 -0.5984>, rot=<0.006 0.0168 -0.0329 -0.9993>}
 middle_01_l {parent=middle_metacarpal_l, loc=<7.3275 -0.2447 0.0129>, rot=<-0.0208 0.0042 0.0858 -0.9961>}
 middle_02_l {parent=middle_01_l, loc=<4.0123 0.1401 0.1175>, rot=<-0.0218 0.0258 0.3512 -0.9357>}
 middle_03_l {parent=middle_02_l, loc=<2.6077 0.0534 0.1373>, rot=<0.0082 0.0188 0.1708 -0.9851>}
 ring_metacarpal_l {parent=hand_l, loc=<2.2074 -0.3625 1.3638>, rot=<-0.0284 0.0722 -0.0458 -0.9959>}
 ring_01_l {parent=ring_metacarpal_l, loc=<7.0295 -0.6197 -0.1446>, rot=<-0.0489 -0.0252 0.1892 -0.9804>}
 ring_02_l {parent=ring_01_l, loc=<3.6174 0.4647 -0.1027>, rot=<0 0 0.2806 -0.9598>}
 ring_03_l {parent=ring_02_l, loc=<2.4152 0.114 0.0199>, rot=<0 0 0.1189 -0.9929>}
 pinky_metacarpal_l {parent=hand_l, loc=<1.7616 -0.9753 3.3814>, rot=<-0.0397 0.0814 -0.069 -0.9935>}
 pinky_01_l {parent=pinky_metacarpal_l, loc=<7.1709 -0.4831 -0.1094>, rot=<-0.0654 -0.0035 0.2242 -0.9723>}
 pinky_02_l {parent=pinky_01_l, loc=<2.8274 0.0893 -0.1043>, rot=<0.0037 0.0011 0.2854 -0.9584>}
 pinky_03_l {parent=pinky_02_l, loc=<2.2046 0.0307 -0.0078>, rot=<-0.0002 0.0098 0.1445 -0.9895>}
 deform_wrist_dn_base_l {parent=hand_l, loc=<0 -2.2342 0>, rot=<0 0 0 -1>}
 deform_wrist_dn_l {parent=deform_wrist_dn_base_l, loc=<0 0 0>, rot=<0 0 0 -1>}
 deform_wrist_up_base_l {parent=hand_l, loc=<0 2.2682 0>, rot=<0 0 0 -1>}
 deform_wrist_up_l {parent=deform_wrist_up_base_l, loc=<0 0 0>, rot=<0 0 0 -1>}
 weapon_l {parent=hand_l, loc=<-0.1974 -3.4111 0.3811>, rot=<0 0 0 1>}
 hand_attach_l {parent=hand_l, loc=<8.5 -3 0>, rot=<0.7071 0.7071 0 0>}
 lowerarm_twist_02_l {parent=lowerarm_l, loc=<9.6199 0 0>, rot=<0 0 0 -1>}
 lowerarm_twist_01_l {parent=lowerarm_l, loc=<17.8549 0 0>, rot=<0 0 0 -1>}
 deform_elbow_in_base_l {parent=lowerarm_l, loc=<0.2112 5.259 0.8274>, rot=<0 0 0 -1>}
 deform_elbow_in_l {parent=deform_elbow_in_base_l, loc=<0 0 0>, rot=<0 0 0 -1>}
 deform_elbow_out_base_l {parent=lowerarm_l, loc=<0 -4 0>, rot=<0 0 0 -1>}
 deform_elbow_out_l {parent=deform_elbow_out_base_l, loc=<0 0 0>, rot=<0 0 0 -1>}
 upperarm_twist_01_l {parent=upperarm_l, loc=<0.0041 0 0>, rot=<0 0 0 -1>}
 deform_upperarm_bk_base_l {parent=upperarm_twist_01_l, loc=<1.0813 -7.2193 -1.9453>, rot=<0 0 0 -1>}
 deform_upperarm_bk_l {parent=deform_upperarm_bk_base_l, loc=<0 0 0>, rot=<0 0 0 1>}
 deform_upperarm_fr_base_l {parent=upperarm_twist_01_l, loc=<-0.3691 7.0506 -2.765>, rot=<0 0 0 -1>}
 deform_upperarm_fr_l {parent=deform_upperarm_fr_base_l, loc=<0 0 0>, rot=<0 0 0 1>}
 deform_upperarm_up_base_l {parent=upperarm_twist_01_l, loc=<-3.2216 -1.1391 3.3311>, rot=<0 0 0 -1>}
 deform_upperarm_up_l {parent=deform_upperarm_up_base_l, loc=<0 0 0>, rot=<0 0 0 1>}
 upperarm_twist_02_l {parent=upperarm_l, loc=<13.7941 0 0>, rot=<0 0 0 -1>}
 base_shoulderPad_l {parent=upperarm_l, loc=<0 0 0>, rot=<0 0 0 1>}
 def_shoulderPad_l {parent=base_shoulderPad_l, loc=<0.9347 0.3245 3.5909>, rot=<0.7085 0.0169 0.005 -0.7055>}
 clavicle_r {parent=spine_05, loc=<4.3575 -0.1466 2.3223>, rot=<0.7289 0.0465 -0.6829 0.009>}
 upperarm_r {parent=clavicle_r, loc=<-15.5111 0 0>, rot=<0.0226 -0.3234 -0.0088 -0.9459>}
 lowerarm_r {parent=upperarm_r, loc=<-26.6514 0 0>, rot=<0.0391 -0.0099 -0.2453 -0.9686>}
 hand_r {parent=lowerarm_r, loc=<-22.3491 0 0>, rot=<-0.7081 -0.0507 0.0975 -0.6975>}
 thumb_01_r {parent=hand_r, loc=<-1.3118 0.3191 2.459>, rot=<0.6546 -0.4042 0.0664 -0.6354>}
 thumb_02_r {parent=thumb_01_r, loc=<-5.9336 0.0448 -0.0401>, rot=<-0.0037 0.0213 0.298 -0.9543>}
 thumb_03_r {parent=thumb_02_r, loc=<-2.2223 -0.3682 0.0907>, rot=<0.0118 -0.013 0.0066 -0.9998>}
 index_metacarpal_r {parent=hand_r, loc=<-1.6535 -0.0183 2.96>, rot=<0.0158 -0.0124 -0.0173 -0.9996>}
 index_01_r {parent=index_metacarpal_r, loc=<-8.3008 -0.0357 0.0524>, rot=<-0.0044 0.0485 0.0924 -0.9945>}
 index_02_r {parent=index_01_r, loc=<-3.3587 0.3004 -0.0452>, rot=<-0.002 0.0027 0.3225 -0.9466>}
 index_03_r {parent=index_02_r, loc=<-2.5578 0.123 0.0252>, rot=<-0.002 0.0067 0.2558 -0.9667>}
 middle_metacarpal_r {parent=hand_r, loc=<-2.1856 0.1725 0.5984>, rot=<0.006 0.0168 -0.0329 -0.9993>}
 middle_01_r {parent=middle_metacarpal_r, loc=<-7.3275 0.2447 -0.0129>, rot=<-0.0208 0.0042 0.0858 -0.9961>}
 middle_02_r {parent=middle_01_r, loc=<-4.0123 -0.1401 -0.1175>, rot=<-0.0218 0.0258 0.3512 -0.9357>}
 middle_03_r {parent=middle_02_r, loc=<-2.6077 -0.0534 -0.1373>, rot=<0.0082 0.0188 0.1708 -0.9851>}
 ring_metacarpal_r {parent=hand_r, loc=<-2.2074 0.3625 -1.3638>, rot=<-0.0284 0.0722 -0.0458 -0.9959>}
 ring_01_r {parent=ring_metacarpal_r, loc=<-7.0295 0.6197 0.1446>, rot=<-0.0489 -0.0252 0.1892 -0.9804>}
 ring_02_r {parent=ring_01_r, loc=<-3.6174 -0.4647 0.1027>, rot=<0 0 0.2806 -0.9598>}
 ring_03_r {parent=ring_02_r, loc=<-2.4152 -0.114 -0.0199>, rot=<0 0 0.1189 -0.9929>}
 pinky_metacarpal_r {parent=hand_r, loc=<-1.7616 0.9753 -3.3814>, rot=<-0.0397 0.0814 -0.069 -0.9935>}
 pinky_01_r {parent=pinky_metacarpal_r, loc=<-7.1709 0.4831 0.1094>, rot=<-0.0654 -0.0035 0.2242 -0.9723>}
 pinky_02_r {parent=pinky_01_r, loc=<-2.8274 -0.0893 0.1043>, rot=<0.0037 0.0011 0.2854 -0.9584>}
 pinky_03_r {parent=pinky_02_r, loc=<-2.2046 -0.0307 0.0078>, rot=<-0.0002 0.0098 0.1445 -0.9895>}
 deform_wrist_dn_base_r {parent=hand_r, loc=<-0.0004 2.2345 -0.0001>, rot=<0 0 0 -1>}
 deform_wrist_dn_r {parent=deform_wrist_dn_base_r, loc=<0 0 0>, rot=<0 0 0 -1>}
 deform_wrist_up_base_r {parent=hand_r, loc=<0.0003 -2.2684 0.0001>, rot=<0 0 0 -1>}
 deform_wrist_up_r {parent=deform_wrist_up_base_r, loc=<0 0 0>, rot=<0 0 0 -1>}
 weapon_r {parent=hand_r, loc=<0.1974 3.4111 -0.3811>, rot=<0 0 0 1>}
 hand_attach_r {parent=hand_r, loc=<-8.5001 2.9999 0>, rot=<0 0 -0.7071 -0.7071>}
 lowerarm_twist_02_r {parent=lowerarm_r, loc=<-9.6199 0 0>, rot=<0 0 0 -1>}
 lowerarm_twist_01_r {parent=lowerarm_r, loc=<-17.8549 0 0>, rot=<0 0 0 -1>}
 deform_elbow_in_base_r {parent=lowerarm_r, loc=<-0.2111 -5.2591 -0.8275>, rot=<0 0 0 -1>}
 deform_elbow_in_r {parent=deform_elbow_in_base_r, loc=<-0.0001 0 0.0002>, rot=<0 0 0 -1>}
 deform_elbow_out_base_r {parent=lowerarm_r, loc=<0 4 -0.0001>, rot=<0 0 0 -1>}
 deform_elbow_out_r {parent=deform_elbow_out_base_r, loc=<0 0 0>, rot=<0 0 0 -1>}
 upperarm_twist_01_r {parent=upperarm_r, loc=<-0.0041 0 0>, rot=<0 0 0 -1>}
 deform_upperarm_bk_base_r {parent=upperarm_twist_01_r, loc=<-1.0814 7.2192 1.9455>, rot=<0 0 0 -1>}
 deform_upperarm_bk_r {parent=deform_upperarm_bk_base_r, loc=<0 0 0>, rot=<0 0 0 1>}
 deform_upperarm_fr_base_r {parent=upperarm_twist_01_r, loc=<0.3689 -7.0506 2.7652>, rot=<0 0 0 -1>}
 deform_upperarm_fr_r {parent=deform_upperarm_fr_base_r, loc=<0 0 0>, rot=<0 0 0 1>}
 deform_upperarm_up_base_r {parent=upperarm_twist_01_r, loc=<3.2214 1.1391 -3.3308>, rot=<0 0 0 -1>}
 deform_upperarm_up_r {parent=deform_upperarm_up_base_r, loc=<0 0 0>, rot=<0 0 0 -1>}
 upperarm_twist_02_r {parent=upperarm_r, loc=<-13.7941 0 0>, rot=<0 0 0 -1>}
 base_shoulderPad_r {parent=upperarm_r, loc=<-0.0001 0 0.0001>, rot=<0 0 0 1>}
 def_shoulderPad_r {parent=base_shoulderPad_r, loc=<-0.9349 -0.3245 -3.5907>, rot=<0.0169 -0.7085 -0.7055 -0.005>}
 deform_pec_aim_base_l {parent=spine_04, loc=<7.8732 12.6388 -2.4123>, rot=<0.0102 -0.4602 0.3979 -0.7936>}
 deform_pec_aim_l {parent=deform_pec_aim_base_l, loc=<0 0 0>, rot=<0 0 0 1>}
 deform_pec_base_l {parent=deform_pec_aim_l, loc=<10.766 0 0>, rot=<0 0 0 1>}
 deform_pec_l {parent=deform_pec_base_l, loc=<0.055 0 0>, rot=<0 0 0 -1>}
 deform_pec_aim_base_r {parent=spine_04, loc=<7.8727 12.6388 2.4123>, rot=<-0.4602 -0.0102 0.7936 0.3979>}
 deform_pec_aim_r {parent=deform_pec_aim_base_r, loc=<0 0 0>, rot=<0 0 0 1>}
 deform_pec_base_r {parent=deform_pec_aim_r, loc=<-10.7662 -0.0006 -0.0007>, rot=<0 0 0 1>}
 deform_pec_r {parent=deform_pec_base_r, loc=<0 0 0>, rot=<0 0 0 -1>}
 def_case_01 {parent=spine_02, loc=<3.4291 7.5217 7.9413>, rot=<0.0135 -0.3009 0.9522 0.0505>}
 def_case_02 {parent=spine_02, loc=<3.3931 6.4838 9.3193>, rot=<0.0135 -0.3009 0.9522 0.0505>}
 def_bag_l {parent=spine_02, loc=<3.0078 1.6289 -10.0318>, rot=<0.0331 -0.7176 -0.6936 -0.0535>}
 def_bag_r {parent=spine_02, loc=<3.0078 1.6289 10.0318>, rot=<0.0337 -0.7387 0.6708 0.0562>}
 spine_05_null {parent=spine_01, type=Null, loc=<31.3352 -1.456 0>, rot=<0 0 -0.024 0.9997>}
 thigh_l {parent=pelvis, loc=<-2.5693 -0.0109 -7.7501>, rot=<0.0012 0.0572 0.0155 -0.9982>}
 calf_l {parent=thigh_l, loc=<-39.0484 0 0>, rot=<0.0006 -0.0161 -0.0391 -0.9991>}
 foot_l {parent=calf_l, loc=<-42.561 0.1077 0.4292>, rot=<0.0418 -0.0455 0.0343 -0.9975>}
 ball_l {parent=foot_l, loc=<-6.1599 13.8264 0.6806>, rot=<0.0001 -0.0001 -0.7063 -0.7079>}
 calf_twist_02_l {parent=calf_l, loc=<-18.8064 -0.0967 0.0397>, rot=<0.0237 0.0095 -0.0035 -0.9997>}
 calf_twist_01_l {parent=calf_l, loc=<-35.813 -0.0967 0.0397>, rot=<0.0237 0.0095 -0.0035 -0.9997>}
 def_foot_corr_bk_l {parent=calf_twist_01_l, loc=<-6.0238 -3.2455 0.7367>, rot=<0.6742 -0.7363 -0.0552 0.0177>}
 def_foot_corr_fr_l {parent=calf_twist_01_l, loc=<-4.595 5.527 0.1642>, rot=<-0.0461 -0.0201 -0.6748 -0.7363>}
 deform_knee_in_base_l {parent=calf_l, loc=<1.298 -5.2957 -0.0496>, rot=<0 0 0 -1>}
 deform_knee_in_l {parent=deform_knee_in_base_l, loc=<0 0 0>, rot=<0 0 0 1>}
 deform_knee_out_base_l {parent=calf_l, loc=<0 4 0>, rot=<0 0 0 -1>}
 deform_knee_out_l {parent=deform_knee_out_base_l, loc=<0 0 0>, rot=<0 0 0 1>}
 thigh_twist_01_l {parent=thigh_l, loc=<-21.6303 0 0>, rot=<0.0013 0 0 -1>}
 def_alt_skirt_corr_l {parent=thigh_twist_01_l, loc=<-2.2197 -9.5225 -0.6826>, rot=<0.026 -0.9933 -0.1125 -0.0051>}
 deform_groin_base_l {parent=thigh_l, loc=<-2.1908 8.1142 -1.7442>, rot=<0 0 0 -1>}
 deform_groin_l {parent=deform_groin_base_l, loc=<0 0 0>, rot=<0 0 0 -1>}
 def_twist_corr_l {parent=thigh_l, loc=<-21.6303 0 0>, rot=<0.0013 0 0 -1>}
 def_twist_corr_bk_l {parent=def_twist_corr_l, loc=<0 -5.8163 0>, rot=<0 0 0 1>}
 calf_l_null {parent=thigh_l, type=Null, loc=<-39.0484 0 0>, rot=<0.0006 -0.0161 -0.0391 -0.9991>}
 thigh_r {parent=pelvis, loc=<-2.5693 -0.0109 7.7501>, rot=<0.0572 -0.0012 0.9982 0.0155>}
 calf_r {parent=thigh_r, loc=<39.0484 0 0>, rot=<0.0006 -0.0161 -0.0391 -0.9991>}
 foot_r {parent=calf_r, loc=<42.561 -0.1077 -0.4292>, rot=<0.0418 -0.0455 0.0343 -0.9975>}
 ball_r {parent=foot_r, loc=<6.1599 -13.8264 -0.6806>, rot=<0.0001 -0.0001 -0.7063 -0.7079>}
 calf_twist_02_r {parent=calf_r, loc=<18.8064 0.0967 -0.0397>, rot=<0.0237 0.0095 -0.0035 -0.9997>}
 calf_twist_01_r {parent=calf_r, loc=<35.813 0.0967 -0.0397>, rot=<0.0237 0.0095 -0.0035 -0.9997>}
 def_foot_corr_fr_r {parent=calf_twist_01_r, loc=<4.595 -5.527 -0.1643>, rot=<0.0201 -0.0461 0.7363 -0.6748>}
 def_foot_corr_bk_r {parent=calf_twist_01_r, loc=<6.0238 3.2455 -0.7367>, rot=<0.7363 0.6742 -0.0177 -0.0552>}
 deform_knee_in_base_r {parent=calf_r, loc=<-1.298 5.2957 0.0496>, rot=<0 0 0 -1>}
 deform_knee_in_r {parent=deform_knee_in_base_r, loc=<0 0 0>, rot=<0 0 0 1>}
 deform_knee_out_base_r {parent=calf_r, loc=<0 -4 0>, rot=<0 0 0 -1>}
 deform_knee_out_r {parent=deform_knee_out_base_r, loc=<0 0 0>, rot=<0 0 0 1>}
 thigh_twist_01_r {parent=thigh_r, loc=<21.6303 0 0>, rot=<0.0013 0 0 -1>}
 def_alt_skirt_corr_r {parent=thigh_twist_01_r, loc=<2.2197 9.5224 0.6826>, rot=<0.9933 0.026 0.0051 -0.1125>}
 deform_groin_base_r {parent=thigh_r, loc=<2.1908 -8.1142 1.7442>, rot=<0 0 0 -1>}
 deform_groin_r {parent=deform_groin_base_r, loc=<0 0 0>, rot=<0 0 0 -1>}
 def_twist_corr_r {parent=thigh_r, loc=<21.6303 0 0>, rot=<0.0013 0 0 -1>}
 def_twist_corr_bk_r {parent=def_twist_corr_r, loc=<0 5.8163 0>, rot=<0 0 0 1>}
 calf_r_null {parent=thigh_r, type=Null, loc=<39.0484 0 0>, rot=<0.0006 -0.0161 -0.0391 -0.9991>}
 deform_glute_base_l {parent=pelvis, loc=<-8.8064 -8.8222 -8.4287>, rot=<0.0275 0.0501 -0.4472 -0.8926>}
 deform_glute_l {parent=deform_glute_base_l, loc=<0 0 0>, rot=<0 0 0 1>}
 deform_glute_base_r {parent=pelvis, loc=<-8.8062 -8.8222 8.4287>, rot=<0.0501 -0.0275 0.8926 -0.4472>}
 deform_glute_r {parent=deform_glute_base_r, loc=<0 0 0>, rot=<0 0 0 1>}
 MSTR_skirt_side_r {parent=pelvis, loc=<1.4216 0.1199 12.8605>, rot=<0.0754 -0.6962 0.7065 0.1022>}
 dyn_skirt_side_02_r {parent=MSTR_skirt_side_r, loc=<0 0 0>, rot=<0 0 0 1>}
 dyn_skirt_side_03_r {parent=dyn_skirt_side_02_r, loc=<19.9554 0 0>, rot=<-0.0203 -0.0049 -0.0145 -0.9997>}
 dyn_skirt_side_05_r {parent=dyn_skirt_side_03_r, loc=<16.7157 0 0>, rot=<0 0 -0.0493 -0.9988>}
 MSTR_skirt_side_l {parent=pelvis, loc=<1.4216 0.1199 -12.8605>, rot=<-0.0754 0.6962 0.7065 0.1022>}
 dyn_skirt_side_02_l {parent=MSTR_skirt_side_l, loc=<0 0 0>, rot=<0 0 0 1>}
 dyn_skirt_side_03_l {parent=dyn_skirt_side_02_l, loc=<19.9555 0 0>, rot=<0.0203 0.0049 -0.0145 -0.9997>}
 dyn_skirt_side_05_l {parent=dyn_skirt_side_03_l, loc=<16.7157 0 0>, rot=<0 0 -0.0493 -0.9988>}
 MSTR_skirt_fr_r {parent=pelvis, loc=<1.7882 8.8486 5.7442>, rot=<0.0467 -0.0858 0.9933 0.062>}
 dyn_skirt_fr_02_r {parent=MSTR_skirt_fr_r, loc=<0 0 0>, rot=<0 0 0 1>}
 dyn_skirt_fr_03_r {parent=dyn_skirt_fr_02_r, loc=<19.5175 0 0>, rot=<0.044 -0.0036 -0.0059 -0.999>}
 dyn_skirt_fr_05_r {parent=dyn_skirt_fr_03_r, loc=<16.5754 0 0>, rot=<0.0436 0 0 -0.999>}
 MSTR_skirt_fr_l {parent=pelvis, loc=<1.7882 8.8486 -5.7442>, rot=<-0.0467 0.0858 0.9933 0.062>}
 dyn_skirt_fr_02_l {parent=MSTR_skirt_fr_l, loc=<0 0 0>, rot=<0 0 0 1>}
 dyn_skirt_fr_03_l {parent=dyn_skirt_fr_02_l, loc=<19.5175 0 0>, rot=<-0.044 0.0036 -0.0059 -0.999>}
 dyn_skirt_fr_05_l {parent=dyn_skirt_fr_03_l, loc=<16.5753 0 0>, rot=<-0.0436 0 0 -0.999>}
 MSTR_skirt_side_rear_l {parent=pelvis, loc=<1.1579 -7.2469 -9.5252>, rot=<0.023 -0.9363 -0.3382 -0.0925>}
 dyn_skirt_side_rear_02_l {parent=MSTR_skirt_side_rear_l, loc=<0 0 0>, rot=<0 0 0 1>}
 dyn_skirt_side_rear_03_l {parent=dyn_skirt_side_rear_02_l, loc=<19.7044 0 0>, rot=<-0.0433 -0.0138 -0.0107 -0.9989>}
 dyn_skirt_side_rear_05_l {parent=dyn_skirt_side_rear_03_l, loc=<16.5834 0 0>, rot=<0.0436 0 0 -0.999>}
 MSTR_skirt_side_rear_r {parent=pelvis, loc=<1.1687 -7.2471 9.5231>, rot=<0.023 -0.9363 0.3382 0.0925>}
 dyn_skirt_side_rear_02_r {parent=MSTR_skirt_side_rear_r, loc=<0 0 0>, rot=<0 0 0 1>}
 dyn_skirt_side_rear_03_r {parent=dyn_skirt_side_rear_02_r, loc=<19.7154 0 0>, rot=<0.0433 0.0138 -0.0107 -0.9989>}
 dyn_skirt_side_rear_05_r {parent=dyn_skirt_side_rear_03_r, loc=<16.5835 0 0>, rot=<-0.0436 0 0 -0.999>}
 MSTR_skirt_fr_rear_l {parent=pelvis, loc=<1.6843 5.7513 -10.8755>, rot=<-0.0652 0.3871 0.9161 0.0818>}
 dyn_skirt_fr_rear_02_l {parent=MSTR_skirt_fr_rear_l, loc=<0 0 0>, rot=<0 0 0 1>}
 dyn_skirt_fr_rear_03_l {parent=dyn_skirt_fr_rear_02_l, loc=<19.8232 0 0>, rot=<0.0291 -0.0005 -0.02 -0.9994>}
 dyn_skirt_fr_rear_05_l {parent=dyn_skirt_fr_rear_03_l, loc=<16.4655 0 0>, rot=<0 0 0 1>}
 MSTR_skirt_fr_rear_r {parent=pelvis, loc=<1.6843 5.7513 10.8755>, rot=<0.0652 -0.3871 0.9161 0.0818>}
 dyn_skirt_fr_rear_02_r {parent=MSTR_skirt_fr_rear_r, loc=<0 0 0>, rot=<0 0 0 1>}
 dyn_skirt_fr_rear_03_r {parent=dyn_skirt_fr_rear_02_r, loc=<19.8232 0 0>, rot=<-0.0291 0.0005 -0.02 -0.9994>}
 dyn_skirt_fr_rear_05_r {parent=dyn_skirt_fr_rear_03_r, loc=<16.4655 0 0>, rot=<0 0 0 1>}
 pelvis_ctrl {controlType=EULER_TRANSFORM, parent=pelvis, type=Control, loc=<0 0 0>, rot=<0 0 0 1>}
 thigh_l_null {parent=pelvis, type=Null, loc=<-2.5693 -0.0109 -7.7501>, rot=<0.0012 0.0572 0.0155 -0.9982>}
 thigh_r_null {parent=pelvis, type=Null, loc=<-2.5693 -0.0109 7.7501>, rot=<0.0572 -0.0012 0.9982 0.0155>}
 MSTR_skirt_bk_l {parent=pelvis, loc=<1.4066 -10.8205 -2.4656>, rot=<-0.0128 -0.9942 -0.0859 -0.0635>}
 dyn_skirt_bk_01_l {parent=MSTR_skirt_bk_l, loc=<0 0 0>, rot=<0 0 0 1>}
 dyn_skirt_bk_02_l {parent=dyn_skirt_bk_01_l, loc=<19.9352 0 0>, rot=<-0.0001 -0.0012 0.0281 -0.9996>}
 dyn_skirt_bk_04_l {parent=dyn_skirt_bk_02_l, loc=<16.5397 0 0>, rot=<0.0436 -0.0009 -0.0201 -0.9988>}
 MSTR_skirt_bk_r {parent=pelvis, loc=<1.4075 -10.8205 2.4655>, rot=<-0.0128 -0.9942 0.0859 0.0635>}
 dyn_skirt_bk_01_r {parent=MSTR_skirt_bk_r, loc=<0 0 0>, rot=<0 0 0 -1>}
 dyn_skirt_bk_02_r {parent=dyn_skirt_bk_01_r, loc=<19.9361 0 0>, rot=<0.0001 0.0012 0.0281 -0.9996>}
 dyn_skirt_bk_04_r {parent=dyn_skirt_bk_02_r, loc=<16.5398 0 0>, rot=<-0.0436 0.0009 -0.0201 -0.9988>}
 ik_hand_root {parent=root, loc=<0 0 0>, rot=<0 0 0 1>}
 ik_hand_gun {parent=ik_hand_root, loc=<-51.9727 4.0883 102.6935>, rot=<0.6084 0.1793 0.4613 -0.6205>}
 ik_hand_r {parent=ik_hand_gun, loc=<0 0 0>, rot=<0 0 0 1>}
 ik_hand_l {parent=ik_hand_gun, loc=<53.0285 82.1781 35.2051>, rot=<-0.5102 -0.7906 -0.3387 0>}
 def_floor {parent=root, loc=<0 0 0>, rot=<0 0 0 1>}
 thigh_l_ctrl {controlType=EULER_TRANSFORM, type=Control, loc=<0 0 0>, rot=<0 0 0 1>}
]
variables [
 Calf_L_BackRotation {type=double, default=0}
 Calf_R_BackRotation {type=double, default=0}
 Leg ROM_L Thigh Raised Fwd Side Inbetween {type=float, default=0}
 Leg ROM_L Thigh Rear Knee Bend {type=double, default=0}
 Leg ROM_R Thigh Raised Fwd Side Inbetween {type=float, default=0}
 Leg ROM_R Thigh Rear Knee Bend {type=double, default=0}
 LegForward_L {type=float, default=0}
 LegForward_R {type=float, default=0}
 PelvisToGround {type=double, default=0}
 Rban {type=bool, default=False}
 SpineToGround {type=double, default=0}
 TacticalSprintingNoRB {type=double, default=0}
 Thigh_L_BackRotation {type=double, default=0}
 Thigh_L_SideRotation {type=double, default=0}
 Thigh_R_BackRotation {type=double, default=0}
 Thigh_R_SideRotation {type=double, default=0}
 isBaseballSliding_CR {type=bool, default=False}
 isCrouching_CR {type=bool, default=False}
 isDBNO {type=bool, default=False}
 isEmoting_CR {type=bool, default=False}
 isIdle {type=bool, default=False}
 isKneeSliding_CR {type=bool, default=False}
 isOnRail_CR {type=bool, default=False}
 isSkydiving_CR {type=bool, default=False}
 jumpsprint {type=double, default=0}
 spineTwist_R {type=double, default=0}
]
