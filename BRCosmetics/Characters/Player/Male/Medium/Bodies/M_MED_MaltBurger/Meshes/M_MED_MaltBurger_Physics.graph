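# Physics asset for the M_MED_MaltBurger male medium body cosmetic, setting up skirt physics around the legs.
# Each side has five three-capsule chains hung from small MSTR_skirt anchor spheres, with limited swing and mostly locked twist.
# Angularly free constraints tie neighboring chain tips from front to back on each side, with no link between left and right.
# Thigh, calf and foot capsules plus a def_floor box are unconstrained bodies, apparently serving as colliders.
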PhysicsAsset /BRCosmetics/Characters/Player/Male/Medium/Bodies/M_MED_MaltBurger/Meshes/M_MED_MaltBurger_Physics
bodies {
 MSTR_skirt_bk_l [{shapeType=Sphere, radius=0.505, shapeName=MSTR_skirt_bk_l_sphere}]
 MSTR_skirt_bk_r [{shapeType=Sphere, radius=0.505, shapeName=MSTR_skirt_bk_r_sphere}]
 MSTR_skirt_fr_l [{shapeType=Sphere, radius=0.505, shapeName=MSTR_skirt_fr_l_sphere}]
 MSTR_skirt_fr_r [{shapeType=Sphere, radius=0.505, shapeName=MSTR_skirt_fr_r_sphere}]
 MSTR_skirt_fr_rear_l [{shapeType=Sphere, radius=0.505, shapeName=MSTR_skirt_fr_rear_l_sphere}]
 MSTR_skirt_fr_rear_r [{shapeType=Sphere, radius=0.505, shapeName=MSTR_skirt_fr_rear_r_sphere}]
 MSTR_skirt_side_l [{shapeType=Sphere, radius=0.505, shapeName=MSTR_skirt_side_l_sphere}]
 MSTR_skirt_side_r [{shapeType=Sphere, radius=0.505, shapeName=MSTR_skirt_side_r_sphere}]
 MSTR_skirt_side_rear_l [{shapeType=Sphere, radius=0.505, shapeName=MSTR_skirt_side_rear_l_sphere}]
 MSTR_skirt_side_rear_r [{shapeType=Sphere, radius=0.505, shapeName=MSTR_skirt_side_rear_r_sphere}]
 calf_l [{
   shapeType=Capsule
   length=33.974
   radius=7.3364
   shapeName=thigh_l_capsule
   center=<-23.0716 -1.4476 0.3149>
   rotation=<rot -0.3103 -93.2487 88.3306>
  }]
 calf_r [{
   shapeType=Capsule
   length=33.974
   radius=7.3364
   shapeName=thigh_r_capsule
   center=<23.0716 1.4476 -0.3149>
   rotation=<rot -0.3103 -93.2487 -91.6694>
  }]
 def_floor [{shapeType=Box, extentX=177.01, extentY=160.01, extentZ=34.01, shapeName=root_box, center=<0 0 -17.0384>, rotation=<rot 0 -90 0>}]
 dyn_skirt_bk_01_l [{
   shapeType=Capsule
   length=5.6711
   radius=8.1212
   shapeName=dyn_skirt_bk_01_l_capsule
   center=<8.3328 -6.1783 1.0314>
   rotation=<rot 5.0436 -90.798 1.9428>
  }]
 dyn_skirt_bk_01_r [{
   shapeType=Capsule
   length=5.6711
   radius=8.1212
   shapeName=dyn_skirt_bk_01_r_capsule
   center=<8.3328 -6.1783 -1.0314>
   rotation=<rot -5.0436 -90.798 178.0572>
  }]
 dyn_skirt_bk_02_l [{
   shapeType=Capsule
   length=9.1094
   radius=7.8746
   shapeName=dyn_skirt_bk_02_l_capsule
   center=<9.0361 -6.9276 0.8024>
   rotation=<rot 4.2599 -91.6533 -175.9144>
  }]
 dyn_skirt_bk_02_r [{
   shapeType=Capsule
   length=9.1094
   radius=7.8746
   shapeName=dyn_skirt_bk_02_r_capsule
   center=<9.0361 -6.9276 -0.8024>
   rotation=<rot 4.2599 -91.6533 4.0856>
  }]
 dyn_skirt_bk_04_l [{
   shapeType=Capsule
   length=13.6004
   radius=6.9766
   shapeName=dyn_skirt_bk_04_l_capsule
   center=<8.3573 -6.002 0.5265>
   rotation=<rot -2.5817 -97.3426 -173.6377>
  }]
 dyn_skirt_bk_04_r [{
   shapeType=Capsule
   length=13.6004
   radius=6.9766
   shapeName=dyn_skirt_bk_04_r_capsule
   center=<8.3573 -6.002 -0.5265>
   rotation=<rot 2.5817 -97.3426 -6.3623>
  }]
 dyn_skirt_fr_02_l [{
   shapeType=Capsule
   length=7.7056
   radius=7.1932
   shapeName=dyn_skirt_fr_02_l_capsule
   center=<8.6466 -5.6945 0.2853>
   rotation=<rot -9.3623 -86.7427 -2.1713>
  }]
 dyn_skirt_fr_02_r [{
   shapeType=Capsule
   length=7.7056
   radius=7.1932
   shapeName=dyn_skirt_fr_02_r_capsule
   center=<8.6466 -5.6945 -0.2853>
   rotation=<rot 13.2199 -86.5925 -177.7993>
  }]
 dyn_skirt_fr_03_l [{
   shapeType=Capsule
   length=10.1891
   radius=6.8034
   shapeName=dyn_skirt_fr_03_l_capsule
   center=<5.8959 -5.0971 0.4856>
   rotation=<rot -4.0774 -86.8808 -0.9248>
  }]
 dyn_skirt_fr_03_r [{
   shapeType=Capsule
   length=10.1891
   radius=6.8034
   shapeName=dyn_skirt_fr_03_r_capsule
   center=<5.8922 -5.0303 -0.4904>
   rotation=<rot 4.0774 -86.8808 -179.0752>
  }]
 dyn_skirt_fr_05_l [{
   shapeType=Capsule
   length=9.1882
   radius=7.8513
   shapeName=dyn_skirt_fr_05_l_capsule
   center=<7.3635 -5.2996 -0.3268>
   rotation=<rot -2.8226 93.5163 -179.9275>
  }]
 dyn_skirt_fr_05_r [{
   shapeType=Capsule
   length=9.1882
   radius=7.8513
   shapeName=dyn_skirt_fr_05_r_capsule
   center=<7.3692 -5.3919 0.3333>
   rotation=<rot 1.3903 93.5182 -0.0724>
  }]
 dyn_skirt_fr_rear_02_l [{
   shapeType=Capsule
   length=7.5
   radius=7.863
   shapeName=dyn_skirt_fr_rear_01_l_capsule
   center=<8.8735 -6.3145 0.3982>
   rotation=<rot -0.4052 90.8598 -174.1953>
  }]
 dyn_skirt_fr_rear_02_r [{
   shapeType=Capsule
   length=7.5
   radius=7.863
   shapeName=dyn_skirt_fr_rear_01_r_capsule
   center=<8.8735 -6.3145 -0.3982>
   rotation=<rot -2.0029 91.1047 -5.8081>
  }]
 dyn_skirt_fr_rear_03_l [{
   shapeType=Capsule
   length=9
   radius=7.363
   shapeName=dyn_skirt_fr_rear_01_l_capsule
   center=<5.8752 -5.8023 0.2987>
   rotation=<rot 3.8527 96.6732 -177.9213>
  }]
 dyn_skirt_fr_rear_03_r [{
   shapeType=Capsule
   length=9
   radius=7.363
   shapeName=dyn_skirt_fr_rear_01_r_capsule
   center=<5.8639 -5.7051 -0.3053>
   rotation=<rot -3.8527 96.6732 -2.0787>
  }]
 dyn_skirt_fr_rear_05_l [{
   shapeType=Capsule
   length=10.8176
   radius=7.0138
   shapeName=dyn_skirt_fr_rear_05_l_capsule
   center=<6.5786 -4.9038 0.7083>
   rotation=<rot -2.8737 -82.624 -1.7863>
  }]
 dyn_skirt_fr_rear_05_r [{
   shapeType=Capsule
   length=10.8176
   radius=7.0138
   shapeName=dyn_skirt_fr_rear_05_r_capsule
   center=<6.5804 -4.9175 -0.7076>
   rotation=<rot 2.8737 -82.624 -178.2137>
  }]
 dyn_skirt_side_02_l [{
   shapeType=Capsule
   length=12.533
   radius=8.0002
   shapeName=dyn_skirt_side_02_l_capsule
   center=<10.4969 -6.6357 -0.0413>
   rotation=<rot -0.2631 102.0803 2.9784>
  }]
 dyn_skirt_side_02_r [{
   shapeType=Capsule
   length=12.533
   radius=8.0002
   shapeName=dyn_skirt_side_02_r_capsule
   center=<10.4969 -6.6357 0.0413>
   rotation=<rot 0.2631 102.0803 177.0216>
  }]
 dyn_skirt_side_03_l [{
   shapeType=Capsule
   length=9.7356
   radius=7.2968
   shapeName=dyn_skirt_side_03_l_capsule
   center=<6.5784 -5.1471 0.1073>
   rotation=<rot 0.0081 101.2149 -0.0435>
  }]
 dyn_skirt_side_03_r [{
   shapeType=Capsule
   length=9.7356
   radius=7.2968
   shapeName=dyn_skirt_side_03_r_capsule
   center=<6.5784 -5.1471 -0.1073>
   rotation=<rot -0.0081 101.2149 -179.9565>
  }]
 dyn_skirt_side_05_l [{
   shapeType=Capsule
   length=11.583
   radius=7.0434
   shapeName=dyn_skirt_side_05_l_capsule
   center=<4.8687 -4.3188 -0.1379>
   rotation=<rot -2.8678 -83.3379 1.6969>
  }]
 dyn_skirt_side_05_r [{
   shapeType=Capsule
   length=11.583
   radius=7.0434
   shapeName=dyn_skirt_side_05_r_capsule
   center=<4.8687 -4.3188 0.1379>
   rotation=<rot 2.8678 -83.3379 178.3031>
  }]
 dyn_skirt_side_rear_02_l [{
   shapeType=Capsule
   length=8.6577
   radius=8.3263
   shapeName=dyn_skirt_bk_02_l_capsule
   center=<8.9342 -7.3422 0.6108>
   rotation=<rot -0.1792 -89.3634 -172.1808>
  }]
 dyn_skirt_side_rear_02_r [{
   shapeType=Capsule
   length=8.6577
   radius=8.3263
   shapeName=dyn_skirt_bk_02_r_capsule
   center=<8.9342 -7.3422 -0.6108>
   rotation=<rot 0.1792 -89.3634 -7.8192>
  }]
 dyn_skirt_side_rear_03_l [{
   shapeType=Capsule
   length=9.451
   radius=7.6506
   shapeName=dyn_skirt_side_rear_03_l_capsule
   center=<7.5716 -6.2522 0.3224>
   rotation=<rot -0.1004 90.5424 -5.3326>
  }]
 dyn_skirt_side_rear_03_r [{
   shapeType=Capsule
   length=9.451
   radius=7.6506
   shapeName=dyn_skirt_side_rear_03_r_capsule
   center=<7.5716 -6.2522 -0.3224>
   rotation=<rot 0.1004 90.5424 174.6674>
  }]
 dyn_skirt_side_rear_05_l [{
   shapeType=Capsule
   length=12.7902
   radius=7.2694
   shapeName=dyn_skirt_side_rear_05_l_capsule
   center=<6.1649 -6.5297 0.3333>
   rotation=<rot 2.4462 -90.2004 174.2729>
  }]
 dyn_skirt_side_rear_05_r [{
   shapeType=Capsule
   length=12.7902
   radius=7.2694
   shapeName=dyn_skirt_side_rear_05_r_capsule
   center=<6.1649 -6.5297 -0.3333>
   rotation=<rot -2.4462 -90.2004 5.7271>
  }]
 foot_l [{
   shapeType=Capsule
   length=4.7485
   radius=8.1838
   shapeName=foot_l_capsule
   center=<-3.1517 2.3023 -0.0009>
   rotation=<rot 0.6041 0.4614 -89.1273>
  }]
 foot_r [{
   shapeType=Capsule
   length=4.7485
   radius=8.1838
   shapeName=foot_r_capsule
   center=<2.9856 -1.7644 0.0109>
   rotation=<rot -1.0095 -72.8343 -89.6723>
  }]
 thigh_l [
  {
   shapeType=Capsule
   length=38.6206
   radius=7.6289
   shapeName=thigh_l_capsule
   center=<-15.6444 0.2473 1.4356>
   rotation=<rot -0.4785 -92.3137 -94.2665>
  }
  {
   shapeType=Capsule
   length=17.3788
   radius=9.9949
   shapeName=thigh_l_capsule
   center=<-29.8244 0.1828 6.3972>
   rotation=<rot -0.4785 -92.3138 -83.2463>
  }
 ]
 thigh_r [
  {
   shapeType=Capsule
   length=38.6206
   radius=7.6289
   shapeName=thigh_r_capsule
   center=<15.6444 -0.2473 -1.4356>
   rotation=<rot -0.4785 -92.3137 85.7335>
  }
  {
   shapeType=Capsule
   length=17.3788
   radius=9.9949
   shapeName=thigh_r_capsule
   center=<29.8244 -0.1828 -6.3972>
   rotation=<rot -0.4785 -92.3138 96.7537>
  }
 ]
}
constraints [
 {
  bone1Name=dyn_skirt_bk_01_l
  bone2Name=MSTR_skirt_bk_l
  swing1LimitDegrees=10
  swing2LimitDegrees=5
  twistLimitDegrees=45
  twistMotion=Locked
 }
 {
  bone1Name=dyn_skirt_bk_02_l
  bone2Name=dyn_skirt_bk_01_l
  swing1LimitDegrees=35
  swing2LimitDegrees=5
  twistLimitDegrees=45
  twistMotion=Locked
 }
 {
  bone1Name=dyn_skirt_bk_04_l
  bone2Name=dyn_skirt_bk_02_l
  swing1LimitDegrees=45
  swing2LimitDegrees=5
  twistLimitDegrees=45
  twistMotion=Locked
 }
 {bone1Name=dyn_skirt_fr_02_l, bone2Name=MSTR_skirt_fr_l, swing1LimitDegrees=7, swing2LimitDegrees=5, twistLimitDegrees=5}
 {bone1Name=dyn_skirt_fr_03_l, bone2Name=dyn_skirt_fr_02_l, swing1LimitDegrees=30, swing2LimitDegrees=10, twistLimitDegrees=10}
 {
  bone1Name=dyn_skirt_fr_05_l
  bone2Name=dyn_skirt_fr_03_l
  swing1LimitDegrees=20
  swing2LimitDegrees=10
  twistLimitDegrees=10
  twistMotion=Locked
 }
 {
  bone1Name=dyn_skirt_fr_rear_02_l
  bone2Name=MSTR_skirt_fr_rear_l
  swing1LimitDegrees=10
  swing2LimitDegrees=10
  twistLimitDegrees=45
  twistMotion=Locked
 }
 {bone1Name=dyn_skirt_fr_rear_03_l, bone2Name=dyn_skirt_fr_rear_02_l, swing1LimitDegrees=25, swing2LimitDegrees=10, twistLimitDegrees=5}
 {
  bone1Name=dyn_skirt_fr_rear_05_l
  bone2Name=dyn_skirt_fr_rear_03_l
  swing1LimitDegrees=15
  swing2LimitDegrees=10
  twistLimitDegrees=45
  twistMotion=Locked
 }
 {
  bone1Name=dyn_skirt_side_02_l
  bone2Name=MSTR_skirt_side_l
  swing1LimitDegrees=15
  swing2LimitDegrees=10
  twistLimitDegrees=45
  twistMotion=Locked
 }
 {
  bone1Name=dyn_skirt_side_03_l
  bone2Name=dyn_skirt_side_02_l
  swing1LimitDegrees=20
  swing2LimitDegrees=10
  twistLimitDegrees=45
  twistMotion=Locked
 }
 {
  bone1Name=dyn_skirt_side_05_l
  bone2Name=dyn_skirt_side_03_l
  swing1LimitDegrees=15
  swing2LimitDegrees=15
  twistLimitDegrees=45
  twistMotion=Locked
 }
 {
  bone1Name=dyn_skirt_side_rear_02_l
  bone2Name=MSTR_skirt_side_rear_l
  swing1LimitDegrees=15
  swing2LimitDegrees=10
  twistLimitDegrees=45
  twistMotion=Locked
 }
 {
  bone1Name=dyn_skirt_side_rear_03_l
  bone2Name=dyn_skirt_side_rear_02_l
  swing1LimitDegrees=35
  swing2LimitDegrees=10
  twistLimitDegrees=45
  twistMotion=Locked
 }
 {
  bone1Name=dyn_skirt_side_rear_05_l
  bone2Name=dyn_skirt_side_rear_03_l
  swing1LimitDegrees=35
  swing2LimitDegrees=10
  twistLimitDegrees=45
  twistMotion=Locked
 }
 {
  bone1Name=dyn_skirt_fr_05_l
  bone2Name=dyn_skirt_fr_rear_05_l
  swing1LimitDegrees=45
  swing1Motion=Free
  swing2LimitDegrees=45
  swing2Motion=Free
  twistLimitDegrees=45
  twistMotion=Free
 }
 {
  bone1Name=dyn_skirt_fr_rear_05_l
  bone2Name=dyn_skirt_side_05_l
  swing1LimitDegrees=45
  swing1Motion=Free
  swing2LimitDegrees=45
  swing2Motion=Free
  twistLimitDegrees=45
  twistMotion=Free
 }
 {
  bone1Name=dyn_skirt_side_05_l
  bone2Name=dyn_skirt_side_rear_05_l
  swing1LimitDegrees=45
  swing1Motion=Free
  swing2LimitDegrees=45
  swing2Motion=Free
  twistLimitDegrees=45
  twistMotion=Free
 }
 {
  bone1Name=dyn_skirt_side_rear_05_l
  bone2Name=dyn_skirt_bk_04_l
  swing1LimitDegrees=45
  swing1Motion=Free
  swing2LimitDegrees=45
  swing2Motion=Free
  twistLimitDegrees=45
  twistMotion=Free
 }
 {
  bone1Name=dyn_skirt_bk_01_r
  bone2Name=MSTR_skirt_bk_r
  swing1LimitDegrees=10
  swing2LimitDegrees=5
  twistLimitDegrees=45
  twistMotion=Locked
 }
 {
  bone1Name=dyn_skirt_bk_02_r
  bone2Name=dyn_skirt_bk_01_r
  swing1LimitDegrees=35
  swing2LimitDegrees=5
  twistLimitDegrees=45
  twistMotion=Locked
 }
 {
  bone1Name=dyn_skirt_bk_04_r
  bone2Name=dyn_skirt_bk_02_r
  swing1LimitDegrees=45
  swing2LimitDegrees=5
  twistLimitDegrees=45
  twistMotion=Locked
 }
 {bone1Name=dyn_skirt_fr_02_r, bone2Name=MSTR_skirt_fr_r, swing1LimitDegrees=7, swing2LimitDegrees=5, twistLimitDegrees=5}
 {
  bone1Name=dyn_skirt_fr_03_r
  bone2Name=dyn_skirt_fr_02_r
  swing1LimitDegrees=30
  swing2LimitDegrees=10
  twistLimitDegrees=10
  twistMotion=Locked
 }
 {
  bone1Name=dyn_skirt_fr_05_r
  bone2Name=dyn_skirt_fr_03_r
  swing1LimitDegrees=20
  swing2LimitDegrees=10
  twistLimitDegrees=10
  twistMotion=Locked
 }
 {
  bone1Name=dyn_skirt_fr_rear_02_r
  bone2Name=MSTR_skirt_fr_rear_r
  swing1LimitDegrees=10
  swing2LimitDegrees=10
  twistLimitDegrees=45
  twistMotion=Locked
 }
 {
  bone1Name=dyn_skirt_fr_rear_03_r
  bone2Name=dyn_skirt_fr_rear_02_r
  swing1LimitDegrees=25
  swing2LimitDegrees=10
  twistLimitDegrees=45
  twistMotion=Locked
 }
 {
  bone1Name=dyn_skirt_fr_rear_05_r
  bone2Name=dyn_skirt_fr_rear_03_r
  swing1LimitDegrees=15
  swing2LimitDegrees=10
  twistLimitDegrees=45
  twistMotion=Locked
 }
 {
  bone1Name=dyn_skirt_side_02_r
  bone2Name=MSTR_skirt_side_r
  swing1LimitDegrees=15
  swing2LimitDegrees=10
  twistLimitDegrees=45
  twistMotion=Locked
 }
 {
  bone1Name=dyn_skirt_side_03_r
  bone2Name=dyn_skirt_side_02_r
  swing1LimitDegrees=20
  swing2LimitDegrees=10
  twistLimitDegrees=45
  twistMotion=Locked
 }
 {
  bone1Name=dyn_skirt_side_05_r
  bone2Name=dyn_skirt_side_03_r
  swing1LimitDegrees=15
  swing2LimitDegrees=15
  twistLimitDegrees=45
  twistMotion=Locked
 }
 {
  bone1Name=dyn_skirt_side_rear_02_r
  bone2Name=MSTR_skirt_side_rear_r
  swing1LimitDegrees=15
  swing2LimitDegrees=10
  twistLimitDegrees=45
  twistMotion=Locked
 }
 {
  bone1Name=dyn_skirt_side_rear_03_r
  bone2Name=dyn_skirt_side_rear_02_r
  swing1LimitDegrees=40
  swing2LimitDegrees=10
  twistLimitDegrees=45
  twistMotion=Locked
 }
 {
  bone1Name=dyn_skirt_side_rear_05_r
  bone2Name=dyn_skirt_side_rear_03_r
  swing1LimitDegrees=35
  swing2LimitDegrees=10
  twistLimitDegrees=45
  twistMotion=Locked
 }
 {
  bone1Name=dyn_skirt_fr_05_r
  bone2Name=dyn_skirt_fr_rear_05_r
  swing1LimitDegrees=45
  swing1Motion=Free
  swing2LimitDegrees=45
  swing2Motion=Free
  twistLimitDegrees=45
  twistMotion=Free
 }
 {
  bone1Name=dyn_skirt_fr_rear_05_r
  bone2Name=dyn_skirt_side_05_r
  swing1LimitDegrees=45
  swing1Motion=Free
  swing2LimitDegrees=45
  swing2Motion=Free
  twistLimitDegrees=45
  twistMotion=Free
 }
 {
  bone1Name=dyn_skirt_side_05_r
  bone2Name=dyn_skirt_side_rear_05_r
  swing1LimitDegrees=45
  swing1Motion=Free
  swing2LimitDegrees=45
  swing2Motion=Free
  twistLimitDegrees=45
  twistMotion=Free
 }
 {
  bone1Name=dyn_skirt_side_rear_05_r
  bone2Name=dyn_skirt_bk_04_r
  swing1LimitDegrees=45
  swing1Motion=Free
  swing2LimitDegrees=45
  swing2Motion=Free
  twistLimitDegrees=45
  twistMotion=Free
 }
]
