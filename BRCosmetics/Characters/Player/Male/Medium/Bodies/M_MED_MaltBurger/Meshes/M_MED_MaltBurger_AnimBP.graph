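# Post-process style AnimBP for the MaltBurger player body cosmetic, layering secondary motion on the pose copied from the parent mesh.
# A Control Rig receives movement and emote states, then constraints blend bag and case bones across pelvis and spine.
# Pelvis-space RigidBody physics follows, then additive offsets on the front skirt bones.
# The event graph computes a gravity override using PureCereal's parameters asset and flags the front-end idle montage for the rig.
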AnimBlueprint /BRCosmetics/Characters/Player/Male/Medium/Bodies/M_MED_MaltBurger/Meshes/M_MED_MaltBurger_AnimBP
target_skeleton=&/M_MED_MaltBurger_Skeleton
graphs [
 AnimGraph {nodes [
   n1 = Control Rig!(
    Source=Copy Pose From Mesh() {bUseAttachedParent=true}
    Alpha
    isCrouching_CR:bool=Get bIsCrouching()
    isBaseballSliding_CR:bool=Get bIsBaseballSliding()
    isKneeSliding_CR:bool=Get bIsKneeSliding()
    isSkydiving_CR:bool=Get bIsSkyDiving()
    isEmoting_CR:bool=Get bIsPlayingEmote()
    isOnRail_CR:bool=Get bIsOnGrindRail()
    TacticalSprintingNoRB:double=Get TacticalSprintingNoRB() -> (TacticalSprintingNoRB:double)
    Rban:bool=Get bRigidBodyEnabled()
    jumpsprint:double=Get JumpingSpring() -> (JumpingSpring:double)
    isIdle:bool=Get isIdleEmote() -> (isIdleEmote:bool)
    isDBNO:bool=Get bIsDBNO()
   ) -> (Pose) {controlRigAssetReference {blueprintRigClass=&/M_MED_MaltBurger_CtrlRig.M_MED_MaltBurger_CtrlRig_C}}
   n2 = "Constraint - def_bag_l (PPPPP)"!(
    ComponentPose=Local To Component(LocalPose=@n1)
    Alpha=1
    ConstraintWeights_0=0.2165
    ConstraintWeights_1:float=0.2141
    ConstraintWeights_2:float=0.3351
    ConstraintWeights_3:float=0.2341
    ConstraintWeights_4:float=0.0001
   ) -> (Pose) {
    boneToModify {boneName=def_bag_l}
    constraintSetup [
     {targetBone {boneName=pelvis}}
     {targetBone {boneName=spine_01}}
     {targetBone {boneName=spine_02}}
     {targetBone {boneName=spine_03}}
     {targetBone {boneName=spine_04}}
    ]
    constraintWeights [1,1,1,1,1]
   }
   n3 = "Constraint - def_bag_r (PPPPP)"!(
    ComponentPose=@n2
    Alpha=1
    ConstraintWeights_0=0.2268
    ConstraintWeights_1:float=0.2135
    ConstraintWeights_2:float=0.3312
    ConstraintWeights_3:float=0.2283
    ConstraintWeights_4:float=0.0003
   ) -> (Pose) {
    boneToModify {boneName=def_bag_r}
    constraintSetup [
     {targetBone {boneName=pelvis}}
     {targetBone {boneName=spine_01}}
     {targetBone {boneName=spine_02}}
     {targetBone {boneName=spine_03}}
     {targetBone {boneName=spine_04}}
    ]
    constraintWeights [1,1,1,1,1]
   }
   n4 = "Constraint - def_case_02 (PPPPP)"!(
    ComponentPose=@n3
    Alpha=1
    ConstraintWeights_0=0.2126
    ConstraintWeights_1:float=0.1841
    ConstraintWeights_2:float=0.3186
    ConstraintWeights_3:float=0.2714
    ConstraintWeights_4:float=0.0133
   ) -> (Pose) {
    boneToModify {boneName=def_case_02}
    constraintSetup [
     {targetBone {boneName=pelvis}}
     {targetBone {boneName=spine_01}}
     {targetBone {boneName=spine_02}}
     {targetBone {boneName=spine_03}}
     {targetBone {boneName=spine_04}}
    ]
    constraintWeights [1,1,1,1,1]
   }
   n5 = "Constraint - def_case_01 (PPPPP)"!(
    ComponentPose=@n4
    Alpha=1
    ConstraintWeights_0=0.2036
    ConstraintWeights_1:float=0.1775
    ConstraintWeights_2:float=0.3244
    ConstraintWeights_3:float=0.28
    ConstraintWeights_4:float=0.0146
   ) -> (Pose) {
    boneToModify {boneName=def_case_01}
    constraintSetup [
     {targetBone {boneName=pelvis}}
     {targetBone {boneName=spine_01}}
     {targetBone {boneName=spine_02}}
     {targetBone {boneName=spine_03}}
     {targetBone {boneName=spine_04}}
    ]
    constraintWeights [1,1,1,1,1]
   }
   n6 = RigidBody!(ComponentPose=@n5, Alpha, OverrideWorldGravity:Vector="0,0,-980", ExternalForce) -> (Pose) {
    bClampLinearTranslationLimitToRefPose=true
    bOverrideWorldGravity=true
    overridePhysicsAsset=&/M_MED_MaltBurger_Physics
    simulationSpace=BaseBoneSpace
    baseBoneRef {boneName=pelvis}
    overrideWorldGravity {z=-980}
    simSpaceSettings {
     dampingAlpha=0
     maxAngularAcceleration=18
     maxAngularVelocity=6
     maxLinearAcceleration=600
     maxLinearVelocity=80
     velocityScaleZ=0.7
     worldAlpha=1
     externalLinearDragV {x=4}
    }
   }
   n7 = "Transform (Modify) Bone - Bone: MSTR_skirt_fr_l"!(ComponentPose=@n6, bAlphaBoolEnabled:bool=True, Translation="0,2,3", Rotation="0,0,10", Scale="1,1,1") -> (Pose) {
    alphaInputType=Bool
    rotationMode=BMM_Additive
    rotationSpace=BCS_BoneSpace
    translationMode=BMM_Additive
    translationSpace=BCS_BoneSpace
    alphaBoolBlend {blendInTime=0.1, blendOutTime=0.1}
    boneToModify {boneName=MSTR_skirt_fr_l}
    rotation {roll=10}
    translation {y=2, z=3}
   }
   n8 = "Transform (Modify) Bone - Bone: MSTR_skirt_fr_r"!(ComponentPose=@n7, bAlphaBoolEnabled:bool=True, Translation="0,1.5,-2", Rotation="0,0,5", Scale="1,1,1") -> (Pose) {
    alphaInputType=Bool
    rotationMode=BMM_Additive
    rotationSpace=BCS_BoneSpace
    translationMode=BMM_Additive
    translationSpace=BCS_BoneSpace
    alphaBoolBlend {blendInTime=0.1, blendOutTime=0.1}
    boneToModify {boneName=MSTR_skirt_fr_r}
    rotation {roll=5}
    translation {y=1.5, z=-2}
   }
   Output Pose(Result=Component To Local(ComponentPose=@n8)) {layerGroup=None}
  ]}
 EventGraph {nodes [
   Event BlueprintUpdateAnimation()
   > n1 = "Gravity Override State Function (Parameter Asset) (Nativized)"(
    CharacterPartAnimInstance=Self-Reference()
    GameTimeInSeconds=GetGameTimeInSeconds()
    Parameters=/BRCosmetics/Characters/Player/Male/Medium/Bodies/M_MED_PureCereal/Meshes/M_MED_PureCereal_Parameters
   )
   > Set Gravity Override(Gravity Override=@n1)
   n2 = GetCurrentActiveMontage!(self:Anim Instance Object Reference=GetAnimInstance(self=Get Custom Base Mesh() -> (Custom Base Mesh:Skeletal Mesh Component Object Reference))) -> (ReturnValue:Anim Montage Object Reference)
   > Set isIdleEmote(isIdleEmote:bool=AND Boolean(
     A="Equal (Object)"(A=@n2, B=/BRCosmetics/Animation/Game/MainPlayer/Menu/BR/CustomSkins/_s40/MaltBurger/MaltBurger_Idle_Montage)
     B=Get bIsInFrontEnd()
    )) -> (Output_Get:bool)
   n3 = Event OnCharacterCustomizationLoaded()
   > Set Custom Base Mesh(Custom Base Mesh:Skeletal Mesh Component Object Reference=@n3.CustomBaseMesh) -> (Output_Get:Skeletal Mesh Component Object Reference)
  ]}
]
variables [
 Custom Base Mesh {type=SkeletalMeshComponent}
 Gravity Override {type=Vector}
 isIdleEmote {type=boolean}
 JumpingSpring {type=number}
 TacticalSprintingNoRB {type=number}
]
