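# Post-mesh Anim Blueprint for the MaltBurger face accessory part, adding secondary motion to its hair curls on top of the attached parent's copied pose.
# A weighted Constraint first blends the back curl bone toward spine, neck and head bones.
# A chain of single-bone Anim Dynamics nodes then springs front, side and back curls within tight angular limits, root-relative with damped component motion.
# The event graph holds no logic.
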AnimBlueprint /BRCosmetics/Characters/Player/Male/Medium/Bodies/M_MED_MaltBurger/Meshes/Parts/M_MED_MaltBurger_FaceAcc_AnimBP
target_skeleton=&/M_MED_MaltBurger_FaceAcc_Skeleton
graphs [
 AnimGraph {nodes [
   n1 = "Constraint - dyn_curl_bk (PPPPP)"!(
    ComponentPose=Local To Component(LocalPose=Copy Pose From Mesh() {bUseAttachedParent=true})
    Alpha=1
    ConstraintWeights_0=0.0027
    ConstraintWeights_1:float=0.0384
    ConstraintWeights_2:float=0.3317
    ConstraintWeights_3:float=0.2864
    ConstraintWeights_4:float=0.3408
   ) -> (Pose) {
    boneToModify {boneName=dyn_curl_bk}
    constraintSetup [
     {targetBone {boneName=spine_04}}
     {targetBone {boneName=spine_05}}
     {targetBone {boneName=neck_01}}
     {targetBone {boneName=neck_02}}
     {targetBone {boneName=head}}
    ]
    constraintWeights [1,1,1,1,1]
   }
   n2 = Anim Dynamics - Bone: dyn_curl_fr_01(ComponentPose=@n1, Alpha=1, ExternalForce="0,0,0") {
    angularDampingOverride=0.99
    angularSpringConstant=150
    bAngularSpring=true
    bOverrideAngularDamping=true
    bUseGravityOverride=true
    simulationSpace=RootRelative
    boundBone {boneName=dyn_curl_fr_01}
    componentAppliedLinearAccClamp=<3000 3000 3000>
    componentLinearAccScale=<0.2 0.2 0.3>
    componentLinearVelScale=<0.2 0.2 0.5>
    gravityOverride {z=-980}
    physicsBodyDefinitions [{
      boundBone {boneName=dyn_curl_fr_01}
      boxExtents=<10 8 8>
      constraintSetup {angularLimitsMax=<0 5 15>, angularLimitsMin=<0 -5 -15>, angularTarget=<1 0 0>}
      localJointOffset=<0.5 0 0>
     }]
    simSpaceSettings {maxAngularAcceleration=200, maxAngularVelocity=40, simSpaceAngularAlpha=0.2}
   }
   n3 = Anim Dynamics - Bone: dyn_curl_fr_02(ComponentPose=@n2, Alpha=1, ExternalForce="0,0,0") {
    angularDampingOverride=0.99
    angularSpringConstant=50
    bAngularSpring=true
    bOverrideAngularDamping=true
    bUseGravityOverride=true
    simulationSpace=RootRelative
    boundBone {boneName=dyn_curl_fr_02}
    componentAppliedLinearAccClamp=<3000 3000 3000>
    componentLinearAccScale=<0.2 0.2 0.3>
    componentLinearVelScale=<0.2 0.2 0.5>
    gravityOverride {z=-980}
    physicsBodyDefinitions [{
      boundBone {boneName=dyn_curl_fr_02}
      boxExtents=<8 6 6>
      constraintSetup {angularLimitsMax=<0 5 10>, angularLimitsMin=<0 -5 -15>, angularTarget=<1 0 0>}
      localJointOffset=<0.5 0 0>
     }]
    simSpaceSettings {maxAngularAcceleration=200, maxAngularVelocity=40, simSpaceAngularAlpha=0.2}
   }
   n4 = Anim Dynamics - Bone: dyn_curl_fr_rear(ComponentPose=@n3, Alpha=1, ExternalForce="0,0,0") {
    angularDampingOverride=0.99
    angularSpringConstant=150
    bAngularSpring=true
    bOverrideAngularDamping=true
    bUseGravityOverride=true
    simulationSpace=RootRelative
    boundBone {boneName=dyn_curl_fr_rear}
    componentAppliedLinearAccClamp=<3000 3000 3000>
    componentLinearAccScale=<0.2 0.2 0.3>
    componentLinearVelScale=<0.2 0.2 0.5>
    gravityOverride {z=-980}
    physicsBodyDefinitions [{
      boundBone {boneName=dyn_curl_fr_rear}
      boxExtents=<10 6 6>
      constraintSetup {angularLimitsMax=<0 5 10>, angularLimitsMin=<0 -10 -10>, angularTarget=<1 0 0>}
      localJointOffset=<0.5 0 0>
     }]
    simSpaceSettings {maxAngularAcceleration=200, maxAngularVelocity=40, simSpaceAngularAlpha=0.2}
   }
   n5 = Anim Dynamics - Bone: dyn_hair_side_l(ComponentPose=@n4, Alpha=1, ExternalForce="0,0,0") {
    angularDampingOverride=0.99
    angularSpringConstant=150
    bAngularSpring=true
    bOverrideAngularDamping=true
    bUseGravityOverride=true
    simulationSpace=RootRelative
    boundBone {boneName=dyn_hair_side_l}
    componentAppliedLinearAccClamp=<3000 3000 3000>
    componentLinearAccScale=<0.2 0.2 0.3>
    componentLinearVelScale=<0.2 0.2 0.5>
    gravityOverride {z=-980}
    physicsBodyDefinitions [{
      boundBone {boneName=dyn_hair_side_l}
      boxExtents=<10 6 6>
      constraintSetup {angularLimitsMax=<5 10 5>, angularLimitsMin=<-5 -5 -1>, angularTarget=<1 0 0>}
      localJointOffset=<0.5 0 0>
     }]
    simSpaceSettings {maxAngularAcceleration=200, maxAngularVelocity=40, simSpaceAngularAlpha=0.2}
   }
   n6 = Anim Dynamics - Bone: dyn_curl_bk_side_01_r(ComponentPose=@n5, Alpha=1, ExternalForce="0,0,0") {
    angularDampingOverride=0.99
    angularSpringConstant=150
    bAngularSpring=true
    bOverrideAngularDamping=true
    bUseGravityOverride=true
    simulationSpace=RootRelative
    boundBone {boneName=dyn_curl_bk_side_01_r}
    componentAppliedLinearAccClamp=<3000 3000 3000>
    componentLinearAccScale=<0.2 0.2 0.3>
    componentLinearVelScale=<0.2 0.2 0.5>
    gravityOverride {z=-980}
    physicsBodyDefinitions [{
      boundBone {boneName=dyn_curl_bk_side_01_r}
      boxExtents=<5 5 5>
      constraintSetup {angularLimitsMax=<0 10 10>, angularLimitsMin=<0 -10 -15>, angularTarget=<1 0 0>}
      localJointOffset=<0.5 0 0>
     }]
    simSpaceSettings {maxAngularAcceleration=200, maxAngularVelocity=40, simSpaceAngularAlpha=0.2}
   }
   n7 = Anim Dynamics - Bone: dyn_curl_bk_side_02_r(ComponentPose=@n6, Alpha=1, ExternalForce="0,0,0") {
    angularDampingOverride=0.99
    angularSpringConstant=150
    bAngularSpring=true
    bOverrideAngularDamping=true
    bUseGravityOverride=true
    simulationSpace=RootRelative
    boundBone {boneName=dyn_curl_bk_side_02_r}
    componentAppliedLinearAccClamp=<3000 3000 3000>
    componentLinearAccScale=<0.2 0.2 0.3>
    componentLinearVelScale=<0.2 0.2 0.5>
    gravityOverride {z=-980}
    physicsBodyDefinitions [{
      boundBone {boneName=dyn_curl_bk_side_02_r}
      boxExtents=<6 6 6>
      constraintSetup {angularLimitsMax=<0 8 8>, angularLimitsMin=<0 -8 -12>, angularTarget=<1 0 0>}
      localJointOffset=<0.5 0 0>
     }]
    simSpaceSettings {maxAngularAcceleration=200, maxAngularVelocity=40, simSpaceAngularAlpha=0.2}
   }
   n8 = Anim Dynamics - Bone: dyn_curl_bk_side_03_r(ComponentPose=@n7, Alpha=1, ExternalForce="0,0,0") {
    angularDampingOverride=0.99
    angularSpringConstant=150
    bAngularSpring=true
    bOverrideAngularDamping=true
    bUseGravityOverride=true
    simulationSpace=RootRelative
    boundBone {boneName=dyn_curl_bk_side_03_r}
    componentAppliedLinearAccClamp=<3000 3000 3000>
    componentLinearAccScale=<0.2 0.2 0.3>
    componentLinearVelScale=<0.2 0.2 0.5>
    gravityOverride {z=-980}
    physicsBodyDefinitions [{
      boundBone {boneName=dyn_curl_bk_side_03_r}
      boxExtents=<4 4 4>
      constraintSetup {angularLimitsMax=<0 7 7>, angularLimitsMin=<0 -7 -10>, angularTarget=<1 0 0>}
      localJointOffset=<0.5 0 0>
     }]
    simSpaceSettings {maxAngularAcceleration=200, maxAngularVelocity=40, simSpaceAngularAlpha=0.2}
   }
   n9 = Anim Dynamics - Bone: dyn_curl_bk_side_01_l(ComponentPose=@n8, Alpha=1, ExternalForce="0,0,0") {
    angularDampingOverride=0.99
    angularSpringConstant=150
    bAngularSpring=true
    bOverrideAngularDamping=true
    bUseGravityOverride=true
    simulationSpace=RootRelative
    boundBone {boneName=dyn_curl_bk_side_01_l}
    componentAppliedLinearAccClamp=<3000 3000 3000>
    componentLinearAccScale=<0.2 0.2 0.3>
    componentLinearVelScale=<0.2 0.2 0.5>
    gravityOverride {z=-980}
    physicsBodyDefinitions [{
      boundBone {boneName=dyn_curl_bk_side_01_l}
      boxExtents=<6 6 6>
      constraintSetup {angularLimitsMax=<0 10 10>, angularLimitsMin=<0 -10 -15>, angularTarget=<1 0 0>}
      localJointOffset=<0.5 0 0>
     }]
    simSpaceSettings {maxAngularAcceleration=200, maxAngularVelocity=40, simSpaceAngularAlpha=0.2}
   }
   n10 = Anim Dynamics - Bone: dyn_curl_bk_side_02_l(ComponentPose=@n9, Alpha=1, ExternalForce="0,0,0") {
    angularDampingOverride=0.99
    angularSpringConstant=150
    bAngularSpring=true
    bOverrideAngularDamping=true
    bUseGravityOverride=true
    simulationSpace=RootRelative
    boundBone {boneName=dyn_curl_bk_side_02_l}
    componentAppliedLinearAccClamp=<3000 3000 3000>
    componentLinearAccScale=<0.2 0.2 0.3>
    componentLinearVelScale=<0.2 0.2 0.5>
    gravityOverride {z=-980}
    physicsBodyDefinitions [{
      boundBone {boneName=dyn_curl_bk_side_02_l}
      boxExtents=<6.2 6.2 6.2>
      constraintSetup {angularLimitsMax=<0 8 8>, angularLimitsMin=<0 -8 -12>, angularTarget=<1 0 0>}
      localJointOffset=<0.5 0 0>
     }]
    simSpaceSettings {maxAngularAcceleration=200, maxAngularVelocity=40, simSpaceAngularAlpha=0.2}
   }
   n11 = Anim Dynamics - Bone: dyn_curl_bk_side_03_l(ComponentPose=@n10, Alpha=1, ExternalForce="0,0,0") {
    angularDampingOverride=0.99
    angularSpringConstant=150
    bAngularSpring=true
    bOverrideAngularDamping=true
    bUseGravityOverride=true
    simulationSpace=RootRelative
    boundBone {boneName=dyn_curl_bk_side_03_l}
    componentAppliedLinearAccClamp=<3000 3000 3000>
    componentLinearAccScale=<0.2 0.2 0.3>
    componentLinearVelScale=<0.2 0.2 0.5>
    gravityOverride {z=-980}
    physicsBodyDefinitions [{
      boundBone {boneName=dyn_curl_bk_side_03_l}
      boxExtents=<6.3 6.3 6.3>
      constraintSetup {angularLimitsMax=<0 7 7>, angularLimitsMin=<0 -7 -10>, angularTarget=<1 0 0>}
      localJointOffset=<0.5 0 0>
     }]
    simSpaceSettings {maxAngularAcceleration=200, maxAngularVelocity=40, simSpaceAngularAlpha=0.2}
   }
   n12 = Anim Dynamics - Bone: dyn_curl_bk_side_04_l(ComponentPose=@n11, Alpha=1, ExternalForce="0,0,0") {
    angularDampingOverride=0.99
    angularSpringConstant=150
    bAngularSpring=true
    bOverrideAngularDamping=true
    bUseGravityOverride=true
    simulationSpace=RootRelative
    boundBone {boneName=dyn_curl_bk_side_04_l}
    componentAppliedLinearAccClamp=<3000 3000 3000>
    componentLinearAccScale=<0.2 0.2 0.3>
    componentLinearVelScale=<0.2 0.2 0.5>
    gravityOverride {z=-980}
    physicsBodyDefinitions [{
      boundBone {boneName=dyn_curl_bk_side_04_l}
      boxExtents=<5.5 5.5 5.5>
      constraintSetup {angularLimitsMax=<0 7 7>, angularLimitsMin=<0 -7 -10>, angularTarget=<1 0 0>}
      localJointOffset=<0.5 0 0>
     }]
    simSpaceSettings {maxAngularAcceleration=200, maxAngularVelocity=40, simSpaceAngularAlpha=0.2}
   }
   n13 = Anim Dynamics - Bone: dyn_curl_bk(ComponentPose=@n12, Alpha=1, ExternalForce="0,0,0") {
    angularDampingOverride=0.99
    angularSpringConstant=150
    bAngularSpring=true
    bOverrideAngularDamping=true
    bUseGravityOverride=true
    simulationSpace=RootRelative
    boundBone {boneName=dyn_curl_bk}
    componentAppliedLinearAccClamp=<3000 3000 3000>
    componentLinearAccScale=<0.2 0.2 0.3>
    componentLinearVelScale=<0.2 0.2 0.5>
    gravityOverride {z=-980}
    physicsBodyDefinitions [{
      boundBone {boneName=dyn_curl_bk}
      boxExtents=<4 4 4>
      constraintSetup {angularLimitsMax=<0 7 7>, angularLimitsMin=<0 -7 -10>, angularTarget=<1 0 0>}
      localJointOffset=<0.5 0 0>
     }]
    simSpaceSettings {maxAngularAcceleration=200, maxAngularVelocity=40, simSpaceAngularAlpha=0.2}
   }
   Output Pose(Result=Component To Local(ComponentPose=@n13)) {layerGroup=None}
  ]}
 EventGraph {nodes [Event BlueprintUpdateAnimation()]}
]
variables []
